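# Physics asset for the cape and hanging garment of the male medium DarkStance Inferno cosmetic body.
# Six cape chains (front, side, back per side) and a garment chain use capsules with angular limits, mostly tighter toward cape tips.
# Free-rotation cross-constraints tie neighboring cape chains into a sheet; one names a nonexistent body, `dyn_cape_b_l_0`.
# Leg and foot capsules, plus a huge zero-length capsule on floor_joint, provide body and ground shapes.
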PhysicsAsset /BRCosmetics/Characters/Player/Male/Medium/Bodies/M_MED_DarkStance_Inferno/Meshes/M_MED_DarkStance_Inferno_Cape_Physics
bodies {
 calf_l [{shapeType=Capsule, length=36.9375, radius=12.875, center=<-24.5597 -2.8398 2.1719>, rotation=<rot 88.0594 -8.7171 -14.8289>}]
 calf_r [
  {shapeType=Capsule, length=38.4375, radius=11.375, center=<24.6324 2.8287 -0.0002>, rotation=<rot 88.5756 -12.1951 -17.4256>}
  {
   shapeType=Capsule
   length=40.255
   radius=5.5935
   shapeName=calf_r_capsule
   center=<18.9509 1.3053 6.081>
   rotation=<rot 82.7057 -3.2427 -3.2652>
  }
 ]
 dyn_cape_b_l_01 [{
   shapeType=Capsule
   length=36.5
   radius=6.5
   shapeName=dyn_cape_b_l_01_capsule
   center=<0 0.0327 3.3262>
   rotation=<rot -0.0002 0.0001 90.5631>
  }]
 dyn_cape_b_l_02 [{shapeType=Capsule, length=27.75, radius=12.2, center=<6 0 10>, rotation=<rot 0 0 90>}]
 dyn_cape_b_l_03 [{shapeType=Capsule, length=27.75, radius=12.2, center=<6 0 9>, rotation=<rot 0 0 90>}]
 dyn_cape_b_l_04 [{shapeType=Capsule, length=27.75, radius=12.2, center=<6 0 8.5>, rotation=<rot 0 0 90>}]
 dyn_cape_b_l_base [{shapeType=Capsule, length=15.5, radius=6.5, shapeName=dyn_cape_b_l_base_capsule, center=<0 0.0327 3.3262>, rotation=<rot -0.0002 0 90.5631>}]
 dyn_cape_b_r_01 [{shapeType=Capsule, length=36.5, radius=6.5, shapeName=dyn_cape_b_r_01_capsule, center=<0 0.0327 3.3262>, rotation=<rot -0.0002 0 90.5631>}]
 dyn_cape_b_r_02 [{shapeType=Capsule, length=27.75, radius=12.2, center=<6 0 10>, rotation=<rot 0 0 90>}]
 dyn_cape_b_r_03 [{shapeType=Capsule, length=27.75, radius=12.2, center=<6 0 9>, rotation=<rot 0 0 90>}]
 dyn_cape_b_r_04 [{shapeType=Capsule, length=27.75, radius=12.2, center=<6 0 8.5>, rotation=<rot 0 0 90>}]
 dyn_cape_b_r_base [{shapeType=Capsule, length=36.5, radius=6.5, shapeName=dyn_cape_b_r_base_capsule, center=<0 0.0327 3.3262>, rotation=<rot -0.0002 0 90.5631>}]
 dyn_cape_f_l_01 [{shapeType=Capsule, length=7.28, radius=10, center=<6 -2.8571 16.2278>, rotation=<rot 90 0 0>}]
 dyn_cape_f_l_02 [{shapeType=Capsule, length=7.25, radius=10, center=<6 0.1655 13.8133>, rotation=<rot 90 0 0>}]
 dyn_cape_f_l_03 [{shapeType=Capsule, length=7.25, radius=10, center=<6 0 12.6322>, rotation=<rot 90 0 0>}]
 dyn_cape_f_l_04 [{shapeType=Capsule, length=7.25, radius=10, center=<6 0.1007 5.0335>, rotation=<rot 41.4303 -75.6697 -80.4056>}]
 dyn_cape_f_l_base [{
   shapeType=Capsule
   length=36.5
   radius=6.5
   shapeName=dyn_cape_f_l_base_capsule
   center=<0 0.0327 3.3262>
   rotation=<rot 0.0001 0.0002 90.5631>
  }]
 dyn_cape_f_r_01 [{shapeType=Capsule, length=7.25, radius=10, center=<6 0.1655 13.8196>, rotation=<rot 90 0 0>}]
 dyn_cape_f_r_02 [{shapeType=Capsule, length=7.25, radius=10, center=<6 0 13>, rotation=<rot 90 0 0>}]
 dyn_cape_f_r_03 [{shapeType=Capsule, length=7.25, radius=10, center=<6 0 12.6322>, rotation=<rot 0 0 90>}]
 dyn_cape_f_r_04 [{shapeType=Capsule, length=7.25, radius=10, center=<7.4707 0 8.7899>, rotation=<rot 0 0 90>}]
 dyn_cape_f_r_base [{
   shapeType=Capsule
   length=36.5
   radius=6.5
   shapeName=dyn_cape_f_r_base_capsule
   center=<0 0.0327 3.3262>
   rotation=<rot 0.0001 -0.0002 90.5632>
  }]
 dyn_cape_s_l_01 [{shapeType=Capsule, length=22.25, radius=10.25, center=<6.0002 2.135 10.4362>, rotation=<rot 0 0 91.5668>}]
 dyn_cape_s_l_02 [{shapeType=Capsule, length=22.25, radius=10.25, center=<6.0002 2.1053 9.3514>, rotation=<rot 0 0 91.5668>}]
 dyn_cape_s_l_03 [{shapeType=Capsule, length=22.25, radius=10.25, center=<6.0002 2.0431 7.0742>, rotation=<rot 0 0 91.5669>}]
 dyn_cape_s_l_04 [{shapeType=Capsule, length=36.1748, radius=12.2, center=<6 0 9.8969>, rotation=<rot 0 0 90>}]
 dyn_cape_s_l_base [{
   shapeType=Capsule
   length=36.5
   radius=6.5
   shapeName=dyn_cape_s_l_base_capsule
   center=<0 0.0327 3.3262>
   rotation=<rot -0.0001 0.0002 90.5631>
  }]
 dyn_cape_s_r_01 [{
   shapeType=Capsule
   length=36.5
   radius=6.5
   shapeName=dyn_cape_s_r_01_capsule
   center=<0 0.0422 4.2923>
   rotation=<rot -0.0001 -0.0002 90.5632>
  }]
 dyn_cape_s_r_02 [{shapeType=Capsule, length=22.25, radius=10.25, center=<6.0002 2.0714 8.1085>, rotation=<rot 0 0 91.5668>}]
 dyn_cape_s_r_03 [{shapeType=Capsule, length=22.25, radius=10.25, center=<6.0002 2.0431 7.0742>, rotation=<rot 0 0 91.5668>}]
 dyn_cape_s_r_04 [{shapeType=Capsule, length=36.1748, radius=12.2, center=<6.2606 0 7.5192>, rotation=<rot 0 0 90>}]
 dyn_cape_s_r_base [{
   shapeType=Capsule
   length=36.5
   radius=6.5
   shapeName=dyn_cape_s_r_base_capsule
   center=<0 0.0327 3.3262>
   rotation=<rot -0.0001 -0.0002 90.5632>
  }]
 dyn_garment_01 [{shapeType=Capsule, length=16, radius=3, shapeName=dyn_garment_01_capsule, center=<2 -2.7 0>}]
 dyn_garment_02 [{shapeType=Capsule, length=12, radius=6, shapeName=dyn_garment_02_capsule, center=<2 -7 0>}]
 dyn_garment_03 [{shapeType=Capsule, length=8.5, radius=6, shapeName=dyn_garment_03_capsule, center=<2 -5.6 0>}]
 dyn_garment_04 [{shapeType=Capsule, length=7.5, radius=4, shapeName=dyn_garment_04_capsule, center=<2 -3.7 0>}]
 dyn_garment_base [{shapeType=Capsule, length=0.505, radius=0.505, shapeName=dyn_garment_base_capsule, rotation=<rot 0 -90 0>}]
 dyn_skul_base [{shapeType=Capsule, length=13.2381, radius=6, shapeName=dyn_skul_base_capsule, center=<0 0 -5.6708>, rotation=<rot 0 -90 0>}]
 floor_joint [{shapeType=Capsule, length=0, radius=324.5692, shapeName=root_capsule, center=<2.6954 -23.3737 -317.2946>, rotation=<rot 0 90 -90>}]
 foot_l [{shapeType=Capsule, length=15, radius=12.75, shapeName=foot_l_capsule, center=<-4.0948 9.0118 -0.1893>, rotation=<rot 0 0 90>}]
 foot_r [{shapeType=Capsule, length=15, radius=12.25, shapeName=foot_r_capsule, center=<4.0948 -8.7964 0.1893>, rotation=<rot 0 0 90>}]
 thigh_l [{
   shapeType=Capsule
   length=26.75
   radius=12.5
   shapeName=thigh_r_capsule
   center=<-17.2569 -2.0936 -1.2977>
   rotation=<rot 85.3118 179.8759 176.9445>
  }]
 thigh_r [{
   shapeType=Capsule
   length=27.75
   radius=12.5
   shapeName=thigh_r_capsule
   center=<15.4135 1.0948 0.028>
   rotation=<rot 86.7774 178.3456 -179.5829>
  }]
}
constraints [
 {
  bone1Name=calf_l
  bone2Name=thigh_l
  swing1LimitDegrees=45
  swing1Motion=Locked
  swing2LimitDegrees=45
  swing2Motion=Locked
  twistLimitDegrees=45
  twistMotion=Locked
 }
 {
  bone1Name=calf_r
  bone2Name=thigh_r
  swing1LimitDegrees=45
  swing1Motion=Locked
  swing2LimitDegrees=45
  swing2Motion=Locked
  twistLimitDegrees=45
  twistMotion=Locked
 }
 {
  bone1Name=foot_r
  bone2Name=calf_r
  swing1LimitDegrees=45
  swing1Motion=Free
  swing2LimitDegrees=45
  swing2Motion=Free
  twistLimitDegrees=45
  twistMotion=Free
 }
 {
  bone1Name=foot_l
  bone2Name=calf_l
  swing1LimitDegrees=45
  swing1Motion=Free
  swing2LimitDegrees=45
  swing2Motion=Free
  twistLimitDegrees=45
  twistMotion=Free
 }
 {bone1Name=dyn_garment_04, bone2Name=dyn_garment_03, swing1LimitDegrees=20, swing2LimitDegrees=30, twistLimitDegrees=20}
 {bone1Name=dyn_garment_02, bone2Name=dyn_garment_01, swing1LimitDegrees=20, swing2LimitDegrees=30, twistLimitDegrees=20}
 {bone1Name=dyn_garment_03, bone2Name=dyn_garment_02, swing1LimitDegrees=20, swing2LimitDegrees=30, twistLimitDegrees=20}
 {bone1Name=dyn_garment_01, bone2Name=dyn_garment_base, swing1LimitDegrees=20, swing2LimitDegrees=30, twistLimitDegrees=20}
 {bone1Name=dyn_cape_b_l_01, bone2Name=dyn_cape_b_l_base, swing1LimitDegrees=5, swing2LimitDegrees=30, twistLimitDegrees=3}
 {bone1Name=dyn_cape_b_l_02, bone2Name=dyn_cape_b_l_01, swing1LimitDegrees=5, swing2LimitDegrees=20, twistLimitDegrees=3}
 {bone1Name=dyn_cape_b_l_03, bone2Name=dyn_cape_b_l_02, swing1LimitDegrees=5, swing2LimitDegrees=20, twistLimitDegrees=3}
 {bone1Name=dyn_cape_b_l_04, bone2Name=dyn_cape_b_l_03, swing1LimitDegrees=5, swing2LimitDegrees=10, twistLimitDegrees=2}
 {bone1Name=dyn_cape_b_r_01, bone2Name=dyn_cape_b_r_base, swing1LimitDegrees=5, swing2LimitDegrees=30, twistLimitDegrees=3}
 {bone1Name=dyn_cape_b_r_02, bone2Name=dyn_cape_b_r_01, swing1LimitDegrees=5, swing2LimitDegrees=20, twistLimitDegrees=3}
 {bone1Name=dyn_cape_b_r_03, bone2Name=dyn_cape_b_r_02, swing1LimitDegrees=5, swing2LimitDegrees=20, twistLimitDegrees=3}
 {bone1Name=dyn_cape_b_r_04, bone2Name=dyn_cape_b_r_03, swing1LimitDegrees=5, swing2LimitDegrees=10, twistLimitDegrees=2}
 {bone1Name=dyn_cape_f_l_01, bone2Name=dyn_cape_f_l_base, swing1LimitDegrees=5, swing2LimitDegrees=30, twistLimitDegrees=3}
 {bone1Name=dyn_cape_f_l_02, bone2Name=dyn_cape_f_l_01, swing1LimitDegrees=5, swing2LimitDegrees=30, twistLimitDegrees=3}
 {bone1Name=dyn_cape_f_l_03, bone2Name=dyn_cape_f_l_02, swing1LimitDegrees=5, swing2LimitDegrees=20, twistLimitDegrees=3}
 {bone1Name=dyn_cape_f_l_04, bone2Name=dyn_cape_f_l_03, swing1LimitDegrees=5, swing2LimitDegrees=15, twistLimitDegrees=3}
 {bone1Name=dyn_cape_f_r_01, bone2Name=dyn_cape_f_r_base, swing1LimitDegrees=5, swing2LimitDegrees=15, twistLimitDegrees=3}
 {bone1Name=dyn_cape_f_r_02, bone2Name=dyn_cape_f_r_01, swing1LimitDegrees=8, swing2LimitDegrees=30, twistLimitDegrees=5}
 {bone1Name=dyn_cape_f_r_03, bone2Name=dyn_cape_f_r_02, swing1LimitDegrees=5, swing2LimitDegrees=20, twistLimitDegrees=3}
 {bone1Name=dyn_cape_f_r_04, bone2Name=dyn_cape_f_r_03, swing1LimitDegrees=5, swing2LimitDegrees=15, twistLimitDegrees=3}
 {bone1Name=dyn_cape_s_l_01, bone2Name=dyn_cape_s_l_base, swing1LimitDegrees=5, swing2LimitDegrees=30, twistLimitDegrees=3}
 {bone1Name=dyn_cape_s_l_02, bone2Name=dyn_cape_s_l_01, swing1LimitDegrees=5, swing2LimitDegrees=30, twistLimitDegrees=3}
 {bone1Name=dyn_cape_s_l_03, bone2Name=dyn_cape_s_l_02, swing1LimitDegrees=5, swing2LimitDegrees=20, twistLimitDegrees=3}
 {bone1Name=dyn_cape_s_l_04, bone2Name=dyn_cape_s_l_03, swing1LimitDegrees=5, swing2LimitDegrees=15, twistLimitDegrees=3}
 {bone1Name=dyn_cape_s_r_01, bone2Name=dyn_cape_s_r_base, swing1LimitDegrees=5, swing2LimitDegrees=30, twistLimitDegrees=3}
 {bone1Name=dyn_cape_s_r_02, bone2Name=dyn_cape_s_r_01, swing1LimitDegrees=5, swing2LimitDegrees=30, twistLimitDegrees=3}
 {bone1Name=dyn_cape_s_r_03, bone2Name=dyn_cape_s_r_02, swing1LimitDegrees=5, swing2LimitDegrees=20, twistLimitDegrees=3}
 {bone1Name=dyn_cape_s_r_04, bone2Name=dyn_cape_s_r_03, swing1LimitDegrees=5, swing2LimitDegrees=15, twistLimitDegrees=3}
 {
  bone1Name=dyn_cape_b_l_04
  bone2Name=dyn_cape_b_r_04
  swing1LimitDegrees=45
  swing1Motion=Free
  swing2LimitDegrees=45
  swing2Motion=Free
  twistLimitDegrees=45
  twistMotion=Free
 }
 {
  bone1Name=dyn_cape_s_l_02
  bone2Name=dyn_cape_b_l_02
  swing1LimitDegrees=45
  swing1Motion=Free
  swing2LimitDegrees=45
  swing2Motion=Free
  twistLimitDegrees=45
  twistMotion=Free
 }
 {
  bone1Name=dyn_cape_s_r_02
  bone2Name=dyn_cape_b_r_02
  swing1LimitDegrees=45
  swing1Motion=Free
  swing2LimitDegrees=45
  swing2Motion=Free
  twistLimitDegrees=45
  twistMotion=Free
 }
 {
  bone1Name=dyn_cape_f_r_02
  bone2Name=dyn_cape_s_r_02
  swing1LimitDegrees=45
  swing1Motion=Free
  swing2LimitDegrees=45
  swing2Motion=Free
  twistLimitDegrees=45
  twistMotion=Free
 }
 {
  bone1Name=dyn_cape_s_r_03
  bone2Name=dyn_cape_f_r_03
  swing1LimitDegrees=45
  swing1Motion=Free
  swing2LimitDegrees=45
  swing2Motion=Free
  twistLimitDegrees=45
  twistMotion=Free
 }
 {
  bone1Name=dyn_cape_s_r_04
  bone2Name=dyn_cape_b_r_04
  swing1LimitDegrees=45
  swing1Motion=Free
  swing2LimitDegrees=45
  swing2Motion=Free
  twistLimitDegrees=45
  twistMotion=Free
 }
 {
  bone1Name=dyn_cape_f_l_02
  bone2Name=dyn_cape_s_l_02
  swing1LimitDegrees=45
  swing1Motion=Free
  swing2LimitDegrees=45
  swing2Motion=Free
  twistLimitDegrees=45
  twistMotion=Free
 }
 {
  bone1Name=dyn_cape_s_l_03
  bone2Name=dyn_cape_f_l_03
  swing1LimitDegrees=45
  swing1Motion=Free
  swing2LimitDegrees=45
  swing2Motion=Free
  twistLimitDegrees=45
  twistMotion=Free
 }
 {
  bone1Name=dyn_cape_s_l_03
  bone2Name=dyn_cape_b_l_0
  swing1LimitDegrees=45
  swing1Motion=Free
  swing2LimitDegrees=45
  swing2Motion=Free
  twistLimitDegrees=45
  twistMotion=Free
 }
 {
  bone1Name=dyn_cape_s_l_04
  bone2Name=dyn_cape_b_l_04
  swing1LimitDegrees=45
  swing1Motion=Free
  swing2LimitDegrees=45
  swing2Motion=Free
  twistLimitDegrees=45
  twistMotion=Free
 }
]
