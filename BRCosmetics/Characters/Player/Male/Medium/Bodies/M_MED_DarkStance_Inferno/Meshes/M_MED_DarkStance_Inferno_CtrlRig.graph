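# Control Rig for the DarkStance Inferno medium male player body, posing its skull, garment and cape dynamic bones around leg motion.
# After the shared DeformRig function, a sequence applies corrective transforms weighted by spherical pose readers on thighs and calf.
# Crouch, movement and jump variables add further offsets, some Kalman-smoothed.
# "RBAN enabled" switches garment correctives; when off, cape chains are parent-constrained to thighs and calves.
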ControlRigBlueprint /BRCosmetics/Characters/Player/Male/Medium/Bodies/M_MED_DarkStance_Inferno/Meshes/M_MED_DarkStance_Inferno_CtrlRig
graphs [
 Add_3_1_1_ContainedGraph {nodes [
   n1 = RigVMFunctionEntryNode()
   RigVMFunctionReturnNode(Result=&RigVMFunction_MathDoubleAdd(A=&RigVMFunction_MathDoubleAdd(A=@n1.A, B=@n1.B).Result, B=@n1.C).Result)
  ]}
 ContainedGraph {nodes [
   n1 = RigVMFunctionEntryNode()
   RigUnit_SequenceAggregate = &RigVMFunction_Sequence(ExecuteContext=@n1.ExecuteContext)
   RigUnit_SequenceAggregate_1 = &RigVMFunction_Sequence(ExecuteContext=@RigUnit_SequenceAggregate.B)
   RigUnit_SequenceAggregate_4 = &RigVMFunction_Sequence(ExecuteContext=@RigUnit_SequenceAggregate_1.B)
   RigUnit_SequenceAggregate_5 = &RigVMFunction_Sequence(ExecuteContext=@RigUnit_SequenceAggregate_4.B)
   RigUnit_SequenceAggregate_6 = &RigVMFunction_Sequence(ExecuteContext=@RigUnit_SequenceAggregate_5.B)
   RigVMFunctionReturnNode(
    A=@RigUnit_SequenceAggregate.A
    B=@RigUnit_SequenceAggregate_1.A
    E=@RigUnit_SequenceAggregate_4.A
    F=@RigUnit_SequenceAggregate_5.A
    G=@RigUnit_SequenceAggregate_6.A
    H=@RigUnit_SequenceAggregate_6.B
   )
  ]}
 RigVMFunctionLibrary {links [], nodes []}
 RigVMFunction_Sequence_1_1_ContainedGraph {nodes [
   n1 = RigVMFunctionEntryNode()
   n2 = &RigVMFunction_Sequence(ExecuteContext=@n1.ExecuteContext)
   n3 = &RigVMFunction_Sequence(ExecuteContext=@n2.B)
   n4 = &RigVMFunction_Sequence(ExecuteContext=@n3.B)
   RigVMFunctionReturnNode(A=@n2.A, B=@n3.A, C=@n4.A, D=@n4.B)
  ]}
 RigVMModel {nodes [
   RigUnit_BeginExecution = &RigUnit_BeginExecution()
   RigUnit_SequenceAggregate = RigVMAggregateNode(ExecuteContext=@RigUnit_BeginExecution.ExecutePin)
   FN_DeformRig_FNC = RigVMFunctionReferenceNode(
    DeformRig_ Curve="(deformRig_l_upperarm_up_alpha_72_#1=False,deformRig_r_upperarm_up_alpha_73_#2=False,deformRig_l_upperarm_fr_alpha_74_#3=False,deformRig_r_upperarm_fr_alpha_75_CB51B8 ...(truncated)"
    DeformRig_Bool="(upperarm_1_#4=True,elbow_3_#5=True,wrist_7_#6=True,groin_8_#7=True,glute_17_#8=true,knee_9_40C408324CA261 ...(truncated)"
    DeformRig_Debug=(upperarm_debug_18_#4=False,wrist_debug_20_#6=False,lat_debug_26_#9=False,groinGlute_debug_27_#7=False)
    DeformRig_Value="(upperarm_up_18_#10=(Rotation=(X=0,Y=0,Z=0,W=1),Translation=(X=0,Y=0,Z=8),Scale3D=(X=0.7,Y=1,Z=1)),upperarm_fwd_fr_19_#11=(Rotation=(X=0,Y=0,Z=0,W=1),Translation=(X=0,Y=10,Z=0),Scale3D=(X=0.5,Y=1,Z= ...(truncated)"
    ExecuteContext=@RigUnit_SequenceAggregate.A
   )
   &RigUnit_AimConstraintLocalSpaceOffset(
    Child=(Type=Bone,Name="dyn_skul_base")
    Filter=(bX=True,bY=True,bZ=True)
    Parents=((Item=(Type=Bone,Name="thigh_r"),Weight=1))
    Weight=1
    WorldUp=(Target=(X=0,Y=0,Z=1),Kind=Direction,Space=(Type=None,Name="None"))
    ExecutePin=@RigUnit_SequenceAggregate.B
   )
   > &RigUnit_ModifyTransforms(
    ItemToModify=((Item=(Type=Bone,Name="dyn_skul_base"),Transform=(Rotation=(X=0.0436,Y=-0.0436,Z=0.0019,W=0.9981),Translation=(X=0,Y=0,Z=0),Scale3D=(X=1,Y=1,Z=1))))
    Weight=1
   )
   > n1 = &RigUnit_SphericalPoseReader(
    ActiveRegionScaleFactors=(PositiveWidth=1,NegativeWidth=1,PositiveHeight=1,NegativeHeight=1)
    ActiveRegionSize=0.1
    Debug=(bDrawDebug=false,DebugScale=25,DebugSegments=20,DebugThickness=0.25)
    DriverAxis=(X=1,Y=0,Z=0)
    DriverItem=(Type=Bone,Name="thigh_r")
    FalloffRegionScaleFactors=(PositiveWidth=1,NegativeWidth=1,PositiveHeight=1,NegativeHeight=1)
    FalloffSize=0.4
    OptionalParentItem=(Type=Bone)
    RotationOffset=(X=-50,Y=30,Z=0)
   )
   > n2 = &RigUnit_ModifyTransforms(
    ItemToModify=((Item=(Type=Bone,Name="dyn_skul_base"),Transform=(Rotation=(X=0.0398,Y=-0.6819,Z=0.0834,W=0.7256),Translation=(X=0,Y=-5,Z=6),Scale3D=(X=1,Y=1,Z=1))))
    Weight=@n1.OutputParam
   )
   RigVMFunction_Sequence = &RigVMFunction_Sequence(ExecuteContext=@n2.ExecutePin)
   RigVMFunction_MathBoolNot = &RigVMFunction_MathBoolNot(Value=&RigVMFunction_MathBoolOr(A=RigVMVariableNode() {variable=iscrouch}.Value, B=RigVMVariableNode() {variable=Is sprinting crouching}.Value).Result)
   n3 = &RigVMFunction_ControlFlowBranch(Condition=@RigVMFunction_MathBoolNot.Result, ExecuteContext=@RigVMFunction_Sequence.A)
   n4 = &RigUnit_SphericalPoseReader(
    ActiveRegionScaleFactors=(PositiveWidth=1,NegativeWidth=1,PositiveHeight=1,NegativeHeight=1)
    ActiveRegionSize=0.02
    Debug=(bDrawDebug=false,DebugScale=25,DebugSegments=20,DebugThickness=0.25)
    DriverAxis=(X=1,Y=0,Z=0)
    DriverItem=(Type=Bone,Name="thigh_r")
    FalloffRegionScaleFactors=(PositiveWidth=1,NegativeWidth=1,PositiveHeight=1,NegativeHeight=1)
    FalloffSize=0.3
    OptionalParentItem=(Type=Bone)
    RotationOffset=(X=-50,Y=-20,Z=40)
    ExecutePin=@n3.True
   )
   > &RigUnit_ModifyTransforms(
    ItemToModify=((Item=(Type=Bone,Name="dyn_skul_base"),Transform=(Rotation=(X=0,Y=-0.2588,Z=0,W=0.9659),Translation=(X=0,Y=0,Z=0),Scale3D=(X=1,Y=1,Z=1))))
    Weight=@n4.OutputParam
   )
   > n5 = &RigUnit_SphericalPoseReader(
    ActiveRegionScaleFactors=(PositiveWidth=1,NegativeWidth=1,PositiveHeight=1,NegativeHeight=1)
    ActiveRegionSize=0
    Debug=(bDrawDebug=false,DebugScale=25,DebugSegments=20,DebugThickness=0.25)
    DriverAxis=(X=1,Y=0,Z=0)
    DriverItem=(Type=Bone,Name="thigh_r")
    FalloffRegionScaleFactors=(PositiveWidth=1,NegativeWidth=1,PositiveHeight=1,NegativeHeight=1)
    FalloffSize=0.05
    OptionalParentItem=(Type=Bone)
    RotationOffset=(X=-30,Y=50,Z=30)
   )
   > &RigUnit_ModifyTransforms(
    ItemToModify=((Item=(Type=Bone,Name="dyn_skul_base"),Transform=(Rotation=(X=0,Y=-0.1736,Z=0,W=0.9848),Translation=(X=0,Y=0,Z=0),Scale3D=(X=1,Y=1,Z=1))))
    Weight=@n5.OutputParam
   )
   RigVMFunction_MathBoolOr = &RigVMFunction_MathBoolOr(A=RigVMVariableNode() {variable=iscrouch}.Value, B=RigVMVariableNode() {variable=Is sprinting crouching}.Value)
   n6 = &RigVMFunction_ControlFlowBranch(Condition=@RigVMFunction_MathBoolOr.Result, ExecuteContext=@RigVMFunction_Sequence.B)
   n7 = &RigUnit_SphericalPoseReader(
    ActiveRegionScaleFactors=(PositiveWidth=1,NegativeWidth=1,PositiveHeight=1,NegativeHeight=1)
    ActiveRegionSize=0.05
    Debug=(bDrawDebug=false,DebugScale=25,DebugSegments=20,DebugThickness=0.25)
    DriverAxis=(X=1,Y=0,Z=0)
    DriverItem=(Type=Bone,Name="thigh_r")
    FalloffRegionScaleFactors=(PositiveWidth=1,NegativeWidth=1,PositiveHeight=1,NegativeHeight=0.5)
    FalloffSize=0.3
    OptionalParentItem=(Type=Bone)
    RotationOffset=(X=-30,Y=-70,Z=0)
    ExecutePin=@n6.True
   )
   > &RigUnit_ModifyTransforms(
    ItemToModify=((Item=(Type=Bone,Name="dyn_skul_base"),Transform=(Rotation=(X=0,Y=-0.1736,Z=0,W=0.9848),Translation=(X=0,Y=0,Z=0),Scale3D=(X=1,Y=1,Z=1))))
    Weight=@n7.OutputParam
   )
   n8 = &RigUnit_SphericalPoseReader(
    ActiveRegionScaleFactors=(PositiveWidth=1,NegativeWidth=0.3,PositiveHeight=1,NegativeHeight=1)
    ActiveRegionSize=0.3
    Debug=(bDrawDebug=false,DebugScale=25,DebugSegments=20,DebugThickness=0.25)
    DriverAxis=(X=-1,Y=0,Z=0)
    DriverItem=(Type=Bone,Name="thigh_l")
    FalloffRegionScaleFactors=(PositiveWidth=1,NegativeWidth=1,PositiveHeight=1,NegativeHeight=1)
    FalloffSize=0.5
    OptionalParentItem=(Type=Bone,Name="spine_02")
    RotationOffset=(X=115,Y=0,Z=-25)
    ExecutePin=@RigUnit_SequenceAggregate.E
   )
   > &RigUnit_ModifyTransforms(
    ItemToModify=((Item=(Type=Bone,Name="dyn_cape_f_l_base"),Transform=(Rotation=(X=0,Y=0,Z=0,W=1),Translation=(X=0,Y=-6,Z=5),Scale3D=(X=1,Y=1,Z=1))))
    ItemToModify.0.Transform=/Script/RigVM.RigVMFunction_MathTransformFromSRT(Location=(X=5,Y=-3,Z=5), Rotation=(X=20,Y=20,Z=0), RotationOrder=XYZ, Scale=(X=1,Y=1,Z=1)) {is=struct RigVMUnitNode}.Transform
    Weight=@n8.OutputParam
   )
   > &RigUnit_ModifyTransforms(
    ItemToModify=((Item=(Type=Bone,Name="dyn_cape_corrective_f_l"),Transform=(Rotation=(X=0,Y=0,Z=0,W=1),Translation=(X=5,Y=-1,Z=2),Scale3D=(X=1,Y=1,Z=1))))
    Weight=@n8.OutputParam
   )
   n9 = &RigVMFunction_ControlFlowBranch(ExecuteContext=@RigUnit_SequenceAggregate.F, Condition=RigVMVariableNode() {variable=RBAN enabled}.Value)
   n10 = &RigUnit_SphericalPoseReader(
    ActiveRegionScaleFactors=(PositiveWidth=1,NegativeWidth=0,PositiveHeight=1,NegativeHeight=1)
    ActiveRegionSize=0.3
    Debug=(bDrawDebug=false,bDraw2D=false,bDrawLocalAxes=false,DebugScale=30,DebugSegments=30,DebugThickness=0.2)
    DriverAxis=(X=-1,Y=0,Z=0)
    DriverItem=(Type=Bone,Name="thigh_l")
    FalloffRegionScaleFactors=(PositiveWidth=1,NegativeWidth=1,PositiveHeight=1,NegativeHeight=1)
    FalloffSize=0.5
    OptionalParentItem=(Type=Bone,Name="pelvis")
    RotationOffset=(X=90,Y=0,Z=-15)
    ExecutePin=@n9.False
   )
   > n11 = &RigUnit_SphericalPoseReader(
    ActiveRegionScaleFactors=(PositiveWidth=1,NegativeWidth=0,PositiveHeight=1,NegativeHeight=1)
    ActiveRegionSize=0.3
    Debug=(bDrawDebug=false,bDraw2D=false,bDrawLocalAxes=false,DebugScale=30,DebugSegments=30,DebugThickness=0.2)
    DriverAxis=(X=1,Y=0,Z=0)
    DriverItem=(Type=Bone,Name="thigh_r")
    FalloffRegionScaleFactors=(PositiveWidth=1,NegativeWidth=1,PositiveHeight=1,NegativeHeight=1)
    FalloffSize=0.5
    OptionalParentItem=(Type=Bone,Name="pelvis")
    RotationOffset=(X=-90,Y=180,Z=-20)
   )
   n12 = RigVMVariableNode() {variable=iscrouch}
   n13 = RigVMVariableNode() {variable=Moving}
   n14 = &RigVMFunction_KalmanFloat(
    BufferSize=16
    Value=&RigVMFunction_MathDoubleAdd(
     A=RigVMDispatchNode(False=0, True=1, Condition=@n12.Value) {template="DISPATCH_RigVMDispatch_If(in Condition,in True,in False,out Result)"}.Result
     B=RigVMDispatchNode(False=0, True=-1, Condition=@n13.Value) {template="DISPATCH_RigVMDispatch_If(in Condition,in True,in False,out Result)"}.Result
    ).Result
   ) {template="KalmanFilter::Execute(in Value,in BufferSize,out Result)"}
   > &RigUnit_ModifyTransforms(
    ItemToModify=((Item=(Type=Bone,Name="dyn_garment_02"),Transform=(Rotation=(X=0.0184,Y=-0.0395,Z=0.4222,W=0.9054),Translation=(X=0,Y=0,Z=0),Scale3D=(X=1,Y=1,Z=1))))
    Weight=@n14.Result
   )
   > &RigUnit_ModifyTransforms(
    ItemToModify="((Item=(Type=Bone,Name=\"dyn_garment_03\"),Transform=(Rotation=(X=0,Y=0,Z=0.1736,W=0.9848),Translation=(X=-2,Y=0,Z=0),Scale3D=(X=1,Y=1,Z=1))),(Item=(Type=Bone,Name=\"None\"),Transform=(Rotation=(X=0,Y=0,Z=0,W=1),Translation=(X=0,Y=0,Z=0),Scale3D=(X=1,Y=1,Z=1)) ...(truncated)"
    WeightMaximum=0.8
    WeightMinimum=0
    Weight=@n14.Result
   )
   > &RigUnit_ModifyTransforms(
    ItemToModify=((Item=(Type=Bone,Name="dyn_garment_base"),Transform=(Rotation=(X=0.8192,Y=0,Z=0,W=0.5736),Translation=(X=0,Y=0,Z=0),Scale3D=(X=1,Y=1,Z=1))))
    WeightMaximum=0.9
    WeightMinimum=0
    Weight=RigVMDispatchNode(
     Condition=/Script/RigVM.RigVMFunction_MathDoubleGreater(A=@n11.OutputParam, B=@n10.OutputParam) {is=struct RigVMUnitNode, template="Greater::Execute(in A,in B,out Result)"}.Result
     True=@n11.OutputParam
     False=@n10.OutputParam
    ) {template="DISPATCH_RigVMDispatch_If(in Condition,in True,in False,out Result)"}.Result
   )
   > &RigUnit_ModifyTransforms(
    ItemToModify=((Item=(Type=Bone,Name="dyn_garment_base_l"),Transform=(Rotation=(X=0,Y=0,Z=0,W=1),Translation=(X=-5,Y=0,Z=0),Scale3D=(X=1,Y=1,Z=1))))
    WeightMaximum=0.9
    WeightMinimum=0
    Weight=@n10.OutputParam
   )
   > &RigUnit_ModifyTransforms(
    ItemToModify=((Item=(Type=Bone,Name="dyn_garment_base_r"),Transform=(Rotation=(X=0.0175,Y=0,Z=0,W=0.9998),Translation=(X=0,Y=0,Z=0),Scale3D=(X=1,Y=1,Z=1))))
    WeightMaximum=0.9
    WeightMinimum=0
    Weight=@n11.OutputParam
   )
   Add_3_1 = RigVMAggregateNode(
    C=-1
    B=RigVMDispatchNode(False=0, True=1, Condition=RigVMVariableNode() {variable=Moving}.Value) {template="DISPATCH_RigVMDispatch_If(in Condition,in True,in False,out Result)"}.Result
    A=RigVMDispatchNode(False=0, True=1, Condition=RigVMVariableNode() {variable=iscrouch}.Value) {template="DISPATCH_RigVMDispatch_If(in Condition,in True,in False,out Result)"}.Result
   )
   > &RigUnit_ModifyTransforms(
    ItemToModify=((Item=(Type=Bone,Name="dyn_garment_03"),Transform=(Rotation=(X=0,Y=0,Z=0.1736,W=0.9848),Translation=(X=-2,Y=0,Z=0),Scale3D=(X=1,Y=1,Z=1))))
    WeightMaximum=0.8
    WeightMinimum=0
    Weight=&RigVMFunction_KalmanFloat(BufferSize=16, Value=@Add_3_1.Result) {template="KalmanFilter::Execute(in Value,in BufferSize,out Result)"}.Result
   )
   n15 = RigVMVariableNode() {variable=Moving}
   n16 = &RigVMFunction_MathDoubleMul(
    B=&RigVMFunction_MathBoolToFloat(Value=&RigVMFunction_MathBoolNot(Value=@n15.Value).Result).Result
    A=&RigVMFunction_MathBoolToFloat(Value=RigVMVariableNode() {variable=iscrouch}.Value).Result
   )
   > &RigUnit_ModifyTransforms(
    ItemToModify=((Item=(Type=Bone,Name="dyn_garment_base"),Transform=(Rotation=(X=-0.1305,Y=0,Z=0,W=0.9914),Translation=(X=0,Y=0,Z=0),Scale3D=(X=1,Y=1,Z=1))))
    WeightMaximum=0.9
    WeightMinimum=0
    Weight=@n16.Result
   )
   > &RigUnit_ModifyTransforms(
    ItemToModify=((Item=(Type=Bone,Name="dyn_garment_03"),Transform=(Rotation=(X=0,Y=0,Z=-0.342,W=0.9397),Translation=(X=-2,Y=0,Z=0),Scale3D=(X=1,Y=1,Z=1))))
    WeightMaximum=0.8
    WeightMinimum=0
    Weight=@n16.Result
   )
   n17 = &RigUnit_SphericalPoseReader(
    ActiveRegionScaleFactors=(PositiveWidth=1,NegativeWidth=0,PositiveHeight=1,NegativeHeight=1)
    ActiveRegionSize=0.3
    Debug=(bDrawDebug=false,bDraw2D=false,bDrawLocalAxes=false,DebugScale=30,DebugSegments=30,DebugThickness=0.2)
    DriverAxis=(X=-1,Y=0,Z=0)
    DriverItem=(Type=Bone,Name="thigh_l")
    FalloffRegionScaleFactors=(PositiveWidth=1,NegativeWidth=0.7,PositiveHeight=1,NegativeHeight=1)
    FalloffSize=0.5
    OptionalParentItem=(Type=Bone,Name="pelvis")
    RotationOffset=(X=90,Y=0,Z=-15)
    ExecutePin=@n9.True
   )
   > n18 = &RigUnit_SphericalPoseReader(
    ActiveRegionScaleFactors=(PositiveWidth=1,NegativeWidth=0,PositiveHeight=1,NegativeHeight=1)
    ActiveRegionSize=0.3
    Debug=(bDrawDebug=false,bDraw2D=false,bDrawLocalAxes=false,DebugScale=30,DebugSegments=30,DebugThickness=0.2)
    DriverAxis=(X=1,Y=0,Z=0)
    DriverItem=(Type=Bone,Name="thigh_r")
    FalloffRegionScaleFactors=(PositiveWidth=1,NegativeWidth=1,PositiveHeight=1,NegativeHeight=1)
    FalloffSize=0.5
    OptionalParentItem=(Type=Bone,Name="pelvis")
    RotationOffset=(X=-90,Y=180,Z=-20)
   )
   n19 = RigVMDispatchNode(
    Condition=&RigVMFunction_MathDoubleGreater(A=@n18.OutputParam, B=@n17.OutputParam) {template="Greater::Execute(in A,in B,out Result)"}.Result
    True=@n18.OutputParam
    False=@n17.OutputParam
   ) {template="DISPATCH_RigVMDispatch_If(in Condition,in True,in False,out Result)"}
   n20 = &RigVMFunction_KalmanFloat(BufferSize=16, Value=@n19.Result) {template="KalmanFilter::Execute(in Value,in BufferSize,out Result)"}
   > &RigUnit_ModifyTransforms(
    ItemToModify=((Item=(Type=Bone,Name="dyn_garment_02"),Transform=(Rotation=(X=0.0184,Y=-0.0395,Z=0.4222,W=0.9054),Translation=(X=0,Y=0,Z=0),Scale3D=(X=1,Y=1,Z=1))))
    Weight=@n20.Result
   )
   > &RigUnit_ModifyTransforms(
    ItemToModify=((Item=(Type=Bone,Name="dyn_garment_03"),Transform=(Rotation=(X=0,Y=0,Z=0.1736,W=0.9848),Translation=(X=-2,Y=0,Z=0),Scale3D=(X=1,Y=1,Z=1))))
    WeightMaximum=0.8
    WeightMinimum=0
    Weight=@n20.Result
   )
   > &RigUnit_ModifyTransforms(
    ItemToModify=((Item=(Type=Bone,Name="dyn_garment_base"),Transform=(Rotation=(X=0.8192,Y=0,Z=0,W=0.5736),Translation=(X=0,Y=0,Z=0),Scale3D=(X=1,Y=1,Z=1))))
    WeightMaximum=0.9
    WeightMinimum=0
    Weight=@n19.Result
   )
   n21 = RigVMVariableNode() {variable=Moving}
   Multiply = &RigVMFunction_MathDoubleMul(
    A=&RigVMFunction_MathBoolToFloat(Value=RigVMVariableNode() {variable=iscrouch}.Value).Result
    B=&RigVMFunction_MathBoolToFloat(Value=&RigVMFunction_MathBoolNot(Value=@n21.Value).Result).Result
   )
   > &RigUnit_ModifyTransforms(
    ItemToModify=((Item=(Type=Bone,Name="dyn_garment_base"),Transform=(Rotation=(X=-0.0261,Y=0.0871,Z=0.0023,W=0.9959),Translation=(X=-5,Y=0,Z=0),Scale3D=(X=1,Y=1,Z=1))))
    WeightMaximum=0.9
    WeightMinimum=0
    Weight=@Multiply.Result
   )
   RigVMFunction_Sequence_1 = RigVMAggregateNode(ExecuteContext=@RigUnit_SequenceAggregate.H)
   &RigUnit_RotationConstraintLocalSpaceOffset(
    AdvancedSettings=(InterpolationType=Average,RotationOrderForFilter=XZY)
    Child=(Type=Bone,Name="dyn_garment_base_l")
    Filter=(bX=True,bY=True,bZ=True)
    Parents=((Item=(Type=Bone,Name="dyn_garment_base"),Weight=1))
    Weight=1
    ExecutePin=@RigVMFunction_Sequence_1.A
   )
   > n22 = &RigUnit_SphericalPoseReader(
    ActiveRegionScaleFactors=(PositiveWidth=1,NegativeWidth=1,PositiveHeight=1,NegativeHeight=1)
    ActiveRegionSize=0.1
    Debug=(bDrawDebug=false,DebugScale=25,DebugSegments=20,DebugThickness=0.25)
    DriverAxis=(X=-1,Y=0,Z=0)
    DriverItem=(Type=Bone,Name="thigh_l")
    FalloffRegionScaleFactors=(PositiveWidth=1,NegativeWidth=1.5,PositiveHeight=1,NegativeHeight=1)
    FalloffSize=0.4
    OptionalParentItem=(Type=Bone)
    RotationOffset=(X=100,Y=0,Z=-30)
   )
   > &RigUnit_ModifyTransforms(
    ItemToModify=((Item=(Type=Bone,Name="dyn_garment_base_l"),Transform=(Rotation=(X=0,Y=0,Z=0,W=1),Translation=(X=0,Y=-8,Z=0),Scale3D=(X=1,Y=1,Z=1))))
    Weight=@n22.OutputParam
   )
   &RigUnit_RotationConstraintLocalSpaceOffset(
    AdvancedSettings=(InterpolationType=Average,RotationOrderForFilter=XZY)
    Child=(Type=Bone,Name="dyn_garment_base_r")
    Filter=(bX=True,bY=True,bZ=True)
    Parents=((Item=(Type=Bone,Name="dyn_garment_base"),Weight=1))
    Weight=1
    ExecutePin=@RigVMFunction_Sequence_1.B
   )
   > n23 = &RigUnit_SphericalPoseReader(
    ActiveRegionScaleFactors=(PositiveWidth=1,NegativeWidth=1,PositiveHeight=1,NegativeHeight=1)
    ActiveRegionSize=0.1
    Debug=(bDrawDebug=false,DebugScale=25,DebugSegments=20,DebugThickness=0.25)
    DriverAxis=(X=1,Y=0,Z=0)
    DriverItem=(Type=Bone,Name="thigh_r")
    FalloffRegionScaleFactors=(PositiveWidth=1.5,NegativeWidth=1,PositiveHeight=1,NegativeHeight=1)
    FalloffSize=0.4
    OptionalParentItem=(Type=Bone)
    RotationOffset=(X=-80,Y=40,Z=-20)
   )
   > &RigUnit_ModifyTransforms(
    ItemToModify=((Item=(Type=Bone,Name="dyn_garment_base_r"),Transform=(Rotation=(X=0,Y=0,Z=0,W=1),Translation=(X=0,Y=-8,Z=0),Scale3D=(X=1,Y=1,Z=1))))
    Weight=@n23.OutputParam
   )
   n24 = &RigUnit_SphericalPoseReader(
    ActiveRegionScaleFactors=(PositiveWidth=1,NegativeWidth=1,PositiveHeight=1,NegativeHeight=1)
    ActiveRegionSize=0.05
    Debug=(bDrawDebug=false,DebugScale=25,DebugSegments=20,DebugThickness=0.25)
    DriverAxis=(X=1,Y=0,Z=0)
    DriverItem=(Type=Bone,Name="calf_l")
    FalloffRegionScaleFactors=(PositiveWidth=1,NegativeWidth=1,PositiveHeight=1,NegativeHeight=1)
    FalloffSize=0.1
    OptionalParentItem=(Type=Bone)
    RotationOffset=(X=-80,Y=40,Z=-120)
    ExecutePin=@RigVMFunction_Sequence_1.C
   )
   > n25 = &RigUnit_SphericalPoseReader(
    ActiveRegionScaleFactors=(PositiveWidth=1,NegativeWidth=1,PositiveHeight=1,NegativeHeight=1)
    ActiveRegionSize=0.03
    Debug=(bDrawDebug=false,DebugScale=25,DebugSegments=20,DebugThickness=0.25)
    DriverAxis=(X=1,Y=0,Z=0)
    DriverItem=(Type=Bone,Name="thigh_r")
    FalloffRegionScaleFactors=(PositiveWidth=1,NegativeWidth=1,PositiveHeight=1,NegativeHeight=1)
    FalloffSize=0.06
    OptionalParentItem=(Type=Bone)
    RotationOffset=(X=60,Y=30,Z=-120)
   )
   n26 = RigVMDispatchNode(
    False=0
    Condition=&RigVMFunction_MathDoubleGreater(B=1.1, A=&RigVMFunction_MathDoubleAdd(B=@n25.OutputParam, A=@n24.OutputParam).Result) {template="Greater::Execute(in A,in B,out Result)"}.Result
    True=@n24.OutputParam
   ) {template="DISPATCH_RigVMDispatch_If(in Condition,in True,in False,out Result)"}
   > &RigUnit_ModifyTransforms(
    ItemToModify=((Item=(Type=Bone,Name="dyn_cape_b_l_base"),Transform=(Rotation=(X=0,Y=-0.2588,Z=0,W=0.9659),Translation=(X=0,Y=0,Z=0),Scale3D=(X=1,Y=1,Z=1))))
    Weight=@n26.Result
   )
   > &RigUnit_ModifyTransforms(
    ItemToModify=((Item=(Type=Bone,Name="dyn_cape_b_r_base"),Transform=(Rotation=(X=0,Y=-0.2588,Z=0,W=0.9659),Translation=(X=0,Y=0,Z=0),Scale3D=(X=1,Y=1,Z=1))))
    Weight=@n26.Result
   )
   > &RigUnit_ModifyTransforms(
    ItemToModify=((Item=(Type=Bone,Name="dyn_cape_f_l_base"),Transform=(Rotation=(X=-0.3829,Y=-0.0729,Z=-0.8726,W=0.2943),Translation=(X=0,Y=0,Z=0),Scale3D=(X=1,Y=1,Z=1))))
    Weight=@n26.Result
   )
   > &RigUnit_ModifyTransforms(
    ItemToModify=((Item=(Type=Bone,Name="dyn_cape_s_l_base"),Transform=(Rotation=(X=-0.1548,Y=-0.6115,Z=-0.2957,W=0.7174),Translation=(X=0,Y=0,Z=0),Scale3D=(X=1,Y=1,Z=1))))
    Weight=@n26.Result
   )
   > &RigUnit_ModifyTransforms(
    ItemToModify=((Item=(Type=Bone,Name="dyn_garment_base"),Transform=(Rotation=(X=-0.3832,Y=0.337,Z=0.0577,W=0.8581),Translation=(X=-5,Y=0,Z=0),Scale3D=(X=1,Y=1,Z=1))))
    Weight=@n26.Result
   )
   > &RigUnit_ModifyTransforms(
    ItemToModify=((Item=(Type=Bone,Name="dyn_skul_base"),Transform=(Rotation=(X=-0.726,Y=0,Z=0,W=0.6877),Translation=(X=-2.6226,Y=0,Z=0),Scale3D=(X=1,Y=1,Z=1))))
    Weight=@n26.Result
   )
   Subtract = &RigVMFunction_MathDoubleSub(A=1, B=&RigVMFunction_MathBoolToFloat(Value=RigVMVariableNode() {variable=RBAN enabled}.Value).Result)
   &RigUnit_ParentConstraint(
    AdvancedSettings=(InterpolationType=Average,RotationOrderForFilter=XZY)
    Child=(Type=Bone,Name="dyn_cape_f_r_02")
    Parents=((Item=(Type=Bone,Name="thigh_r"),Weight=1))
    ExecutePin=@RigVMFunction_Sequence_1.D
    Weight=@Subtract.Result
   )
   > &RigUnit_ParentConstraint(
    AdvancedSettings=(InterpolationType=Average,RotationOrderForFilter=XZY)
    Child=(Type=Bone,Name="dyn_cape_f_r_03")
    Filter=(TranslationFilter=(bX=false,bY=false,bZ=false),RotationFilter=(bX=True,bY=True,bZ=True),ScaleFilter=(bX=True,bY=True,bZ=True))
    Parents=((Item=(Type=Bone,Name="calf_r"),Weight=1))
    Weight=@Subtract.Result
   )
   > &RigUnit_ParentConstraint(
    AdvancedSettings=(InterpolationType=Average,RotationOrderForFilter=XZY)
    Child=(Type=Bone,Name="dyn_cape_s_r_02")
    Parents=((Item=(Type=Bone,Name="thigh_r"),Weight=1))
    Weight=@Subtract.Result
   )
   > &RigUnit_ParentConstraint(
    AdvancedSettings=(InterpolationType=Average,RotationOrderForFilter=XZY)
    Child=(Type=Bone,Name="dyn_cape_s_r_03")
    Filter=(TranslationFilter=(bX=false,bY=false,bZ=false),RotationFilter=(bX=True,bY=True,bZ=True),ScaleFilter=(bX=True,bY=True,bZ=True))
    Parents=((Item=(Type=Bone,Name="calf_r"),Weight=1))
    Weight=@Subtract.Result
   )
   > &RigUnit_ParentConstraint(
    AdvancedSettings=(InterpolationType=Average,RotationOrderForFilter=XZY)
    Child=(Type=Bone,Name="dyn_cape_b_r_02")
    Filter=(TranslationFilter=(bX=true,bY=true,bZ=true),RotationFilter=(bX=True,bY=True,bZ=True),ScaleFilter=(bX=True,bY=True,bZ=True))
    Parents=((Item=(Type=Bone,Name="thigh_r"),Weight=1))
    Weight=@Subtract.Result
   )
   > &RigUnit_ParentConstraint(
    AdvancedSettings=(InterpolationType=Average,RotationOrderForFilter=XZY)
    Child=(Type=Bone,Name="dyn_cape_b_r_03")
    Filter=(TranslationFilter=(bX=false,bY=false,bZ=false),RotationFilter=(bX=True,bY=True,bZ=True),ScaleFilter=(bX=True,bY=True,bZ=True))
    Parents=((Item=(Type=Bone,Name="calf_r"),Weight=1))
    Weight=@Subtract.Result
   )
   > &RigUnit_ParentConstraint(
    AdvancedSettings=(InterpolationType=Average,RotationOrderForFilter=XZY)
    Child=(Type=Bone,Name="dyn_cape_f_l_02")
    Parents=((Item=(Type=Bone,Name="thigh_l"),Weight=1))
    Weight=@Subtract.Result
   )
   > &RigUnit_ParentConstraint(
    AdvancedSettings=(InterpolationType=Average,RotationOrderForFilter=XZY)
    Child=(Type=Bone,Name="dyn_cape_f_l_03")
    Filter=(TranslationFilter=(bX=false,bY=false,bZ=false),RotationFilter=(bX=True,bY=True,bZ=True),ScaleFilter=(bX=True,bY=True,bZ=True))
    Parents=((Item=(Type=Bone,Name="calf_l"),Weight=1))
    Weight=@Subtract.Result
   )
   > &RigUnit_ParentConstraint(
    AdvancedSettings=(InterpolationType=Average,RotationOrderForFilter=XZY)
    Child=(Type=Bone,Name="dyn_cape_s_l_02")
    Parents=((Item=(Type=Bone,Name="thigh_l"),Weight=1))
    Weight=@Subtract.Result
   )
   > &RigUnit_ParentConstraint(
    AdvancedSettings=(InterpolationType=Average,RotationOrderForFilter=XZY)
    Child=(Type=Bone,Name="dyn_cape_s_l_03")
    Filter=(TranslationFilter=(bX=false,bY=false,bZ=false),RotationFilter=(bX=True,bY=True,bZ=True),ScaleFilter=(bX=True,bY=True,bZ=True))
    Parents=((Item=(Type=Bone,Name="calf_l"),Weight=1))
    Weight=@Subtract.Result
   )
   > &RigUnit_ParentConstraint(
    AdvancedSettings=(InterpolationType=Average,RotationOrderForFilter=XZY)
    Child=(Type=Bone,Name="dyn_cape_b_l_02")
    Filter=(TranslationFilter=(bX=true,bY=true,bZ=true),RotationFilter=(bX=True,bY=True,bZ=True),ScaleFilter=(bX=True,bY=True,bZ=True))
    Parents=((Item=(Type=Bone,Name="thigh_l"),Weight=1))
    Weight=@Subtract.Result
   )
   > &RigUnit_ParentConstraint(
    AdvancedSettings=(InterpolationType=Average,RotationOrderForFilter=XZY)
    Child=(Type=Bone,Name="dyn_cape_b_l_03")
    Filter=(TranslationFilter=(bX=false,bY=false,bZ=false),RotationFilter=(bX=True,bY=True,bZ=True),ScaleFilter=(bX=True,bY=True,bZ=True))
    Parents=((Item=(Type=Bone,Name="calf_l"),Weight=1))
    Weight=@Subtract.Result
   )
   n27 = &RigVMFunction_MathQuaternionFromEuler(Euler=(X=0,Y=-50,Z=0))
   RigVMFunction_MathBoolToFloat = &RigVMFunction_MathBoolToFloat(Value=RigVMVariableNode() {variable=Is jump}.Value)
   KalmanFilter = &RigVMFunction_KalmanFloat(BufferSize=10, Value=@RigVMFunction_MathBoolToFloat.Result) {template="KalmanFilter::Execute(in Value,in BufferSize,out Result)"}
   AlphaInterp = &RigVMFunction_AlphaInterp(InterpSpeedDecreasing=1, InterpSpeedIncreasing=1, Scale=1, bClampResult=False, bInterpResult=False, Value=@KalmanFilter.Result)
   &RigUnit_ModifyTransforms(
    ItemToModify=((Item=(Type=Bone,Name="dyn_cape_b_l_base"),Transform=(Rotation=(X=0,Y=-0.1736,Z=0,W=0.9848),Translation=(X=0,Y=0,Z=0),Scale3D=(X=1,Y=1,Z=1))))
    ItemToModify.0.Transform.Rotation=@n27.Result
    ExecutePin=@RigUnit_SequenceAggregate.G
    Weight=@AlphaInterp.Result
   )
   > &RigUnit_ModifyTransforms(
    ItemToModify=((Item=(Type=Bone,Name="dyn_cape_b_r_base"),Transform=(Rotation=(X=0,Y=-0.1736,Z=0,W=0.9848),Translation=(X=0,Y=0,Z=0),Scale3D=(X=1,Y=1,Z=1))))
    ItemToModify.0.Transform.Rotation=@n27.Result
    Weight=@AlphaInterp.Result
   )
   > &RigUnit_ModifyTransforms(
    ItemToModify=((Item=(Type=Bone,Name="dyn_cape_s_l_base"),Transform=(Rotation=(X=0.0873,Y=-0.2787,Z=-0.0982,W=0.9513),Translation=(X=0,Y=0,Z=0),Scale3D=(X=1,Y=1,Z=1))))
    Weight=@AlphaInterp.Result
   )
   > &RigUnit_ModifyTransforms(
    ItemToModify=((Item=(Type=Bone,Name="dyn_cape_s_r_base"),Transform=(Rotation=(X=0,Y=-0.2588,Z=0,W=0.9659),Translation=(X=0,Y=0,Z=0),Scale3D=(X=1,Y=1,Z=1))))
    Weight=@AlphaInterp.Result
   )
  ]}
]
hierarchy [
 root {loc=<0 0 0>, rot=<0 0 0 1>}
 pelvis {parent=root, loc=<0 -0.7071 93.1191>, rot=<0.0127 0.707 -0.0127 -0.707>}
 spine_01 {parent=pelvis, loc=<6.7494 -0.6388 0>, rot=<0 0 0.034 -0.9994>}
 spine_02 {parent=spine_01, loc=<7.2518 -0.035 0>, rot=<0 0 0.0238 -0.9997>}
 spine_03 {parent=spine_02, loc=<6.0104 0 0>, rot=<0 0 0.0149 -0.9999>}
 spine_04 {parent=spine_03, loc=<8.1024 0 0>, rot=<0 0 -0.0132 -0.9999>}
 spine_05 {parent=spine_04, loc=<10.0146 0 0>, rot=<0 0 -0.0015 -1>}
 attach_cape {parent=spine_05, loc=<9.017 -11.714 0>, rot=<0.0616 -0.7044 -0.0616 -0.7044>}
 neck_01 {parent=spine_05, loc=<11.3138 0.1529 0>, rot=<0 0 -0.2003 -0.9797>}
 neck_02 {parent=neck_01, loc=<4.5627 0 0>, rot=<0 0 0.0504 -0.9987>}
 head {parent=neck_02, loc=<5.3914 0 0>, rot=<0 0 0.115 -0.9934>}
 attach_backpack {parent=spine_05, loc=<-0.123 -13.473 0>, rot=<-0.0093 -0.707 0.0093 -0.707>}
 clavicle_l {parent=spine_05, loc=<4.3575 -0.1466 -2.3223>, rot=<0.0465 -0.7289 -0.009 -0.6829>}
 upperarm_l {parent=clavicle_l, loc=<15.5111 0 0>, rot=<0.0226 -0.3234 -0.0088 -0.9459>}
 lowerarm_l {parent=upperarm_l, loc=<26.6514 0 0>, rot=<0.0391 -0.0099 -0.2453 -0.9686>}
 hand_l {parent=lowerarm_l, loc=<22.3491 0 0>, rot=<-0.7081 -0.0507 0.0975 -0.6975>}
 thumb_01_l {parent=hand_l, loc=<1.3118 -0.3191 -2.459>, rot=<0.6546 -0.4042 0.0664 -0.6354>}
 thumb_02_l {parent=thumb_01_l, loc=<5.9336 -0.0448 0.0401>, rot=<-0.0037 0.0213 0.298 -0.9543>}
 thumb_03_l {parent=thumb_02_l, loc=<2.2223 0.3682 -0.0907>, rot=<0.0118 -0.013 0.0066 -0.9998>}
 index_metacarpal_l {parent=hand_l, loc=<1.6535 0.0183 -2.96>, rot=<0.0158 -0.0124 -0.0173 -0.9996>}
 index_01_l {parent=index_metacarpal_l, loc=<8.3008 0.0357 -0.0524>, rot=<-0.0044 0.0485 0.0924 -0.9945>}
 index_02_l {parent=index_01_l, loc=<3.3587 -0.3004 0.0452>, rot=<-0.002 0.0027 0.3225 -0.9466>}
 index_03_l {parent=index_02_l, loc=<2.5578 -0.123 -0.0252>, rot=<-0.002 0.0067 0.2558 -0.9667>}
 middle_metacarpal_l {parent=hand_l, loc=<2.1856 -0.1725 -0.5984>, rot=<0.006 0.0168 -0.0329 -0.9993>}
 middle_01_l {parent=middle_metacarpal_l, loc=<7.3275 -0.2447 0.0129>, rot=<-0.0208 0.0042 0.0858 -0.9961>}
 middle_02_l {parent=middle_01_l, loc=<4.0123 0.1401 0.1175>, rot=<-0.0218 0.0258 0.3512 -0.9357>}
 middle_03_l {parent=middle_02_l, loc=<2.6077 0.0534 0.1373>, rot=<0.0082 0.0188 0.1708 -0.9851>}
 ring_metacarpal_l {parent=hand_l, loc=<2.2074 -0.3625 1.3638>, rot=<-0.0284 0.0722 -0.0458 -0.9959>}
 ring_01_l {parent=ring_metacarpal_l, loc=<7.0295 -0.6197 -0.1446>, rot=<-0.0489 -0.0252 0.1892 -0.9804>}
 ring_02_l {parent=ring_01_l, loc=<3.6174 0.4647 -0.1027>, rot=<0 0 0.2806 -0.9598>}
 ring_03_l {parent=ring_02_l, loc=<2.4152 0.114 0.0199>, rot=<0 0 0.1189 -0.9929>}
 pinky_metacarpal_l {parent=hand_l, loc=<1.7616 -0.9753 3.3814>, rot=<-0.0397 0.0814 -0.069 -0.9935>}
 pinky_01_l {parent=pinky_metacarpal_l, loc=<7.1709 -0.4831 -0.1094>, rot=<-0.0654 -0.0035 0.2242 -0.9723>}
 pinky_02_l {parent=pinky_01_l, loc=<2.8274 0.0893 -0.1043>, rot=<0.0037 0.0011 0.2854 -0.9584>}
 pinky_03_l {parent=pinky_02_l, loc=<2.2046 0.0307 -0.0078>, rot=<-0.0002 0.0098 0.1445 -0.9895>}
 weapon_l {parent=hand_l, loc=<-0.1974 -3.4111 0.3811>, rot=<0 0 0 1>}
 hand_attach_l {parent=hand_l, loc=<8.5 -3 0>, rot=<0.7071 0.7071 0 0>}
 deform_wrist_dn_base_l {parent=hand_l, loc=<0 -2.2342 0>, rot=<0 0 0 -1>}
 deform_wrist_dn_l {parent=deform_wrist_dn_base_l, loc=<0 0 0>, rot=<0 0 0 1>}
 deform_wrist_up_base_l {parent=hand_l, loc=<0 2.2682 0>, rot=<0 0 0 -1>}
 deform_wrist_up_l {parent=deform_wrist_up_base_l, loc=<0 0 0>, rot=<0 0 0 1>}
 lowerarm_twist_02_l {parent=lowerarm_l, loc=<9.6199 0 0>, rot=<0 0 0 -1>}
 lowerarm_twist_01_l {parent=lowerarm_l, loc=<17.8549 0 0>, rot=<0 0 0 -1>}
 deform_elbow_in_base_l {parent=lowerarm_l, loc=<0.2112 5.259 0.8274>, rot=<0 0 0 -1>}
 deform_elbow_in_l {parent=deform_elbow_in_base_l, loc=<0 0 0>, rot=<0 0 0 1>}
 deform_elbow_out_base_l {parent=lowerarm_l, loc=<0 -4 0>, rot=<0 0 0 -1>}
 deform_elbow_out_l {parent=deform_elbow_out_base_l, loc=<0 0 0>, rot=<0 0 0 1>}
 upperarm_twist_01_l {parent=upperarm_l, loc=<0.0041 0 0>, rot=<0 0 0 1>}
 deform_upperarm_bk_base_l {parent=upperarm_twist_01_l, loc=<1.0813 -7.2193 -1.9453>, rot=<0 0 0 1>}
 deform_upperarm_bk_l {parent=deform_upperarm_bk_base_l, loc=<0 0 0>, rot=<0 0 0 1>}
 deform_upperarm_fr_base_l {parent=upperarm_twist_01_l, loc=<-0.3691 7.0506 -2.765>, rot=<0 0 0 1>}
 deform_upperarm_fr_l {parent=deform_upperarm_fr_base_l, loc=<0 0 0>, rot=<0 0 0 1>}
 deform_upperarm_up_base_l {parent=upperarm_twist_01_l, loc=<-3.2216 -1.1391 3.3311>, rot=<0 0 0 1>}
 deform_upperarm_up_l {parent=deform_upperarm_up_base_l, loc=<0 0 0>, rot=<0 0 0 1>}
 upperarm_twist_02_l {parent=upperarm_l, loc=<13.7941 0 0>, rot=<0 0 0 1>}
 clavicle_r {parent=spine_05, loc=<4.3575 -0.1466 2.3223>, rot=<0.7289 0.0465 -0.6829 0.009>}
 upperarm_r {parent=clavicle_r, loc=<-15.5111 0 0>, rot=<0.0226 -0.3234 -0.0088 -0.9459>}
 lowerarm_r {parent=upperarm_r, loc=<-26.6514 0 0>, rot=<0.0391 -0.0099 -0.2453 -0.9686>}
 hand_r {parent=lowerarm_r, loc=<-22.3491 0 0>, rot=<-0.7081 -0.0507 0.0975 -0.6975>}
 thumb_01_r {parent=hand_r, loc=<-1.3118 0.3191 2.459>, rot=<0.6546 -0.4042 0.0664 -0.6354>}
 thumb_02_r {parent=thumb_01_r, loc=<-5.9336 0.0448 -0.0401>, rot=<-0.0037 0.0213 0.298 -0.9543>}
 thumb_03_r {parent=thumb_02_r, loc=<-2.2223 -0.3682 0.0907>, rot=<0.0118 -0.013 0.0066 -0.9998>}
 index_metacarpal_r {parent=hand_r, loc=<-1.6535 -0.0183 2.96>, rot=<0.0158 -0.0124 -0.0173 -0.9996>}
 index_01_r {parent=index_metacarpal_r, loc=<-8.3008 -0.0357 0.0524>, rot=<-0.0044 0.0485 0.0924 -0.9945>}
 index_02_r {parent=index_01_r, loc=<-3.3587 0.3004 -0.0452>, rot=<-0.002 0.0027 0.3225 -0.9466>}
 index_03_r {parent=index_02_r, loc=<-2.5578 0.123 0.0252>, rot=<-0.002 0.0067 0.2558 -0.9667>}
 middle_metacarpal_r {parent=hand_r, loc=<-2.1856 0.1725 0.5984>, rot=<0.006 0.0168 -0.0329 -0.9993>}
 middle_01_r {parent=middle_metacarpal_r, loc=<-7.3275 0.2447 -0.0129>, rot=<-0.0208 0.0042 0.0858 -0.9961>}
 middle_02_r {parent=middle_01_r, loc=<-4.0123 -0.1401 -0.1175>, rot=<-0.0218 0.0258 0.3512 -0.9357>}
 middle_03_r {parent=middle_02_r, loc=<-2.6077 -0.0534 -0.1373>, rot=<0.0082 0.0188 0.1708 -0.9851>}
 ring_metacarpal_r {parent=hand_r, loc=<-2.2074 0.3625 -1.3638>, rot=<-0.0284 0.0722 -0.0458 -0.9959>}
 ring_01_r {parent=ring_metacarpal_r, loc=<-7.0295 0.6197 0.1446>, rot=<-0.0489 -0.0252 0.1892 -0.9804>}
 ring_02_r {parent=ring_01_r, loc=<-3.6174 -0.4647 0.1027>, rot=<0 0 0.2806 -0.9598>}
 ring_03_r {parent=ring_02_r, loc=<-2.4152 -0.114 -0.0199>, rot=<0 0 0.1189 -0.9929>}
 pinky_metacarpal_r {parent=hand_r, loc=<-1.7616 0.9753 -3.3814>, rot=<-0.0397 0.0814 -0.069 -0.9935>}
 pinky_01_r {parent=pinky_metacarpal_r, loc=<-7.1709 0.4831 0.1094>, rot=<-0.0654 -0.0035 0.2242 -0.9723>}
 pinky_02_r {parent=pinky_01_r, loc=<-2.8274 -0.0893 0.1043>, rot=<0.0037 0.0011 0.2854 -0.9584>}
 pinky_03_r {parent=pinky_02_r, loc=<-2.2046 -0.0307 0.0078>, rot=<-0.0002 0.0098 0.1445 -0.9895>}
 weapon_r {parent=hand_r, loc=<0.1974 3.4111 -0.3811>, rot=<0 0 0 1>}
 hand_attach_r {parent=hand_r, loc=<-8.5001 2.9999 0>, rot=<0 0 -0.7071 -0.7071>}
 deform_wrist_dn_base_r {parent=hand_r, loc=<-0.0004 2.2345 -0.0001>, rot=<0 0 0 -1>}
 deform_wrist_dn_r {parent=deform_wrist_dn_base_r, loc=<0 0 0>, rot=<0 0 0 1>}
 deform_wrist_up_base_r {parent=hand_r, loc=<0.0003 -2.2684 0.0001>, rot=<0 0 0 -1>}
 deform_wrist_up_r {parent=deform_wrist_up_base_r, loc=<0 0 0>, rot=<0 0 0 1>}
 lowerarm_twist_02_r {parent=lowerarm_r, loc=<-9.6199 0 0>, rot=<0 0 0 -1>}
 lowerarm_twist_01_r {parent=lowerarm_r, loc=<-17.8549 0 0>, rot=<0 0 0 -1>}
 deform_elbow_in_base_r {parent=lowerarm_r, loc=<-0.2111 -5.2591 -0.8275>, rot=<0 0 0 -1>}
 deform_elbow_in_r {parent=deform_elbow_in_base_r, loc=<-0.0001 0 0.0002>, rot=<0 0 0 1>}
 deform_elbow_out_base_r {parent=lowerarm_r, loc=<0 4 -0.0001>, rot=<0 0 0 -1>}
 deform_elbow_out_r {parent=deform_elbow_out_base_r, loc=<0 0 0>, rot=<0 0 0 1>}
 upperarm_twist_01_r {parent=upperarm_r, loc=<-0.0041 0 0>, rot=<0 0 0 -1>}
 deform_upperarm_bk_base_r {parent=upperarm_twist_01_r, loc=<-1.0814 7.2192 1.9455>, rot=<0 0 0 -1>}
 deform_upperarm_bk_r {parent=deform_upperarm_bk_base_r, loc=<0 0 0>, rot=<0 0 0 1>}
 deform_upperarm_fr_base_r {parent=upperarm_twist_01_r, loc=<0.3689 -7.0506 2.7652>, rot=<0 0 0 -1>}
 deform_upperarm_fr_r {parent=deform_upperarm_fr_base_r, loc=<0 0 0>, rot=<0 0 0 1>}
 deform_upperarm_up_base_r {parent=upperarm_twist_01_r, loc=<3.2214 1.1391 -3.3308>, rot=<0 0 0 -1>}
 deform_upperarm_up_r {parent=deform_upperarm_up_base_r, loc=<0 0 0>, rot=<0 0 0 -1>}
 upperarm_twist_02_r {parent=upperarm_r, loc=<-13.7941 0 0>, rot=<0 0 0 -1>}
 cape_spine_corrective {parent=spine_04, loc=<6 -13.4678 0>, rot=<0 -0.7071 0 -0.7071>}
 dyn_cape_corrective_f_l_base {parent=spine_03, loc=<2.2817 8.1354 -12.0857>, rot=<-0.1933 -0.6802 -0.2665 -0.655>}
 dyn_cape_corrective_f_l {parent=dyn_cape_corrective_f_l_base, loc=<0 0 0>, rot=<0 0 0 1>}
 dyn_cape_corrective_f_r_base {parent=spine_03, loc=<2.7277 8.3752 12.5554>, rot=<0.2665 -0.655 0.1933 -0.6802>}
 dyn_cape_corrective_f_r {parent=dyn_cape_corrective_f_r_base, loc=<0 0 0>, rot=<0 0 0 1>}
 thigh_l {parent=pelvis, loc=<-2.5693 -0.0109 -7.7501>, rot=<0.0012 0.0572 0.0155 -0.9982>}
 calf_l {parent=thigh_l, loc=<-39.0484 0 0>, rot=<0.0006 -0.0161 -0.0391 -0.9991>}
 foot_l {parent=calf_l, loc=<-42.561 0.1077 0.4292>, rot=<0.0418 -0.0455 0.0343 -0.9975>}
 ball_l {parent=foot_l, loc=<-6.1599 13.8264 0.6806>, rot=<0.0001 -0.0001 -0.7063 -0.7079>}
 calf_twist_02_l {parent=calf_l, loc=<-18.8064 -0.0967 0.0397>, rot=<0.0237 0.0095 -0.0035 -0.9997>}
 calf_twist_01_l {parent=calf_l, loc=<-35.813 -0.0967 0.0397>, rot=<0.0237 0.0095 -0.0035 -0.9997>}
 deform_knee_in_base_l {parent=calf_l, loc=<1.298 -5.2957 -0.0496>, rot=<0 0 0 -1>}
 deform_knee_in_l {parent=deform_knee_in_base_l, loc=<0 0 0>, rot=<0 0 0 1>}
 deform_knee_out_base_l {parent=calf_l, loc=<0 4 0>, rot=<0 0 0 -1>}
 deform_knee_out_l {parent=deform_knee_out_base_l, loc=<0 0 0>, rot=<0 0 0 1>}
 thigh_twist_01_l {parent=thigh_l, loc=<-21.6303 0 0>, rot=<0.0013 0 0 -1>}
 deform_groin_base_l {parent=thigh_l, loc=<-2.1908 8.1142 -1.7442>, rot=<0 0 0 -1>}
 deform_groin_l {parent=deform_groin_base_l, loc=<0 0 0>, rot=<0 0 0 1>}
 thigh_r {parent=pelvis, loc=<-2.5693 -0.0109 7.7501>, rot=<0.0572 -0.0012 0.9982 0.0155>}
 calf_r {parent=thigh_r, loc=<39.0484 0 0>, rot=<0.0006 -0.0161 -0.0391 -0.9991>}
 foot_r {parent=calf_r, loc=<42.561 -0.1077 -0.4292>, rot=<0.0418 -0.0455 0.0343 -0.9975>}
 ball_r {parent=foot_r, loc=<6.1599 -13.8264 -0.6806>, rot=<0.0001 -0.0001 -0.7063 -0.7079>}
 calf_twist_02_r {parent=calf_r, loc=<18.8064 0.0967 -0.0397>, rot=<0.0237 0.0095 -0.0035 -0.9997>}
 calf_twist_01_r {parent=calf_r, loc=<35.813 0.0967 -0.0397>, rot=<0.0237 0.0095 -0.0035 -0.9997>}
 deform_knee_in_base_r {parent=calf_r, loc=<-1.298 5.2957 0.0496>, rot=<0 0 0 -1>}
 deform_knee_in_r {parent=deform_knee_in_base_r, loc=<0 0 0>, rot=<0 0 0 1>}
 deform_knee_out_base_r {parent=calf_r, loc=<0 -4 0>, rot=<0 0 0 -1>}
 deform_knee_out_r {parent=deform_knee_out_base_r, loc=<0 0 0>, rot=<0 0 0 1>}
 thigh_twist_01_r {parent=thigh_r, loc=<21.6303 0 0>, rot=<0.0013 0 0 -1>}
 deform_groin_base_r {parent=thigh_r, loc=<2.1908 -8.1142 1.7442>, rot=<0 0 0 -1>}
 deform_groin_r {parent=deform_groin_base_r, loc=<0 0 0>, rot=<0 0 0 1>}
 dyn_cape_f_r_base {parent=pelvis, loc=<13.2077 3.0395 14.755>, rot=<0.0776 0.2843 0.9522 -0.0809>}
 dyn_cape_f_r_01 {parent=dyn_cape_f_r_base, loc=<0 0 0>, rot=<0 0 0 1>}
 dyn_cape_f_r_02 {parent=dyn_cape_f_r_01, loc=<17.799 0 0>, rot=<0 -0.0017 0.0048 -1>}
 dyn_cape_f_r_03 {parent=dyn_cape_f_r_02, loc=<20 0 0>, rot=<0.0446 -0.0059 -0.0036 -0.999>}
 dyn_cape_f_r_04 {parent=dyn_cape_f_r_03, loc=<21.8726 0 0>, rot=<0.0664 0.0087 0.0298 -0.9973>}
 dyn_cape_s_r_base {parent=pelvis, loc=<12.4116 -6.1564 15.3014>, rot=<0.0583 -0.1672 0.9825 -0.0577>}
 dyn_cape_s_r_01 {parent=dyn_cape_s_r_base, loc=<0.0003 0 0>, rot=<0 0 0 1>}
 dyn_cape_s_r_02 {parent=dyn_cape_s_r_01, loc=<13.5443 0 0>, rot=<0.0007 0.018 0.0021 -0.9998>}
 dyn_cape_s_r_03 {parent=dyn_cape_s_r_02, loc=<23.4406 0.0391 -0.3402>, rot=<0.0852 -0.0176 0.0164 -0.9961>}
 dyn_cape_s_r_04 {parent=dyn_cape_s_r_03, loc=<20.0247 0 0>, rot=<0.0018 -0.0163 -0.0134 -0.9998>}
 dyn_cape_b_r_base {parent=pelvis, loc=<12.3415 -13.5366 6>, rot=<0.067 -0.6228 0.7793 -0.0171>}
 dyn_cape_b_r_01 {parent=dyn_cape_b_r_base, loc=<0.0005 0 0>, rot=<0 0 0 1>}
 dyn_cape_b_r_02 {parent=dyn_cape_b_r_01, loc=<15.2519 -0.0001 0>, rot=<0 0.0092 0 -1>}
 dyn_cape_b_r_03 {parent=dyn_cape_b_r_02, loc=<21.7441 0 -0.1428>, rot=<0 0 0 1>}
 dyn_cape_b_r_04 {parent=dyn_cape_b_r_03, loc=<20.0949 0 0>, rot=<0 0 0 1>}
 dyn_cape_b_l_base {parent=pelvis, loc=<12.3404 -13.5525 -6>, rot=<0.0198 -0.7671 0.6383 -0.0615>}
 dyn_cape_b_l_01 {parent=dyn_cape_b_l_base, loc=<-0.0023 0.0029 -0.0152>, rot=<0.061 0 0 -0.9981>}
 dyn_cape_b_l_02 {parent=dyn_cape_b_l_01, loc=<16.4234 0 0>, rot=<0 0.0092 0 -1>}
 dyn_cape_b_l_03 {parent=dyn_cape_b_l_02, loc=<20.5731 0 -0.1212>, rot=<0 0 0 1>}
 dyn_cape_b_l_04 {parent=dyn_cape_b_l_03, loc=<20.7941 0 0>, rot=<0 0 0 1>}
 dyn_cape_s_l_base {parent=pelvis, loc=<12.4444 -5.6449 -15.5997>, rot=<0.0494 -0.9841 0.1581 -0.0642>}
 dyn_cape_s_l_01 {parent=dyn_cape_s_l_base, loc=<-0.0002 0 0>, rot=<0 0 0 1>}
 dyn_cape_s_l_02 {parent=dyn_cape_s_l_01, loc=<16.668 0 0>, rot=<0.0007 0.018 0.0021 -0.9998>}
 dyn_cape_s_l_03 {parent=dyn_cape_s_l_02, loc=<20.3195 0.0262 -0.2277>, rot=<-0.1138 -0.041 0.0142 -0.9926>}
 dyn_cape_s_l_04 {parent=dyn_cape_s_l_03, loc=<19.8309 0 0>, rot=<0.073 0.0241 0.0186 -0.9969>}
 dyn_cape_f_l_base {parent=pelvis, loc=<13.2549 2.7094 -15.0329>, rot=<0.0805 -0.9398 -0.3243 -0.0713>}
 dyn_cape_f_l_01 {parent=dyn_cape_f_l_base, loc=<0 0 0>, rot=<0 0 0 1>}
 dyn_cape_f_l_02 {parent=dyn_cape_f_l_01, loc=<17.7992 0 0>, rot=<-0.001 0.0052 0.0024 -1>}
 dyn_cape_f_l_03 {parent=dyn_cape_f_l_02, loc=<20 0.025 -0.0543>, rot=<-0.1149 -0.0168 0.011 -0.9932>}
 dyn_cape_f_l_04 {parent=dyn_cape_f_l_03, loc=<22.0912 0 0>, rot=<-0.0353 0.0241 0.0443 -0.9981>}
 dyn_garment_base {parent=pelvis, loc=<2.7105 8.9197 0>, rot=<-0.0127 -0.707 0.0127 -0.707>}
 dyn_garment_01 {parent=dyn_garment_base, loc=<0 0 0>, rot=<0.7069 0.0156 -0.7069 -0.0156>}
 dyn_garment_02 {parent=dyn_garment_01, loc=<13.2851 0.3141 0>, rot=<0 0 0 1>}
 dyn_garment_03 {parent=dyn_garment_02, loc=<17.3732 0.1134 0>, rot=<0 0 0 1>}
 dyn_garment_04 {parent=dyn_garment_03, loc=<14.6565 0.0585 0>, rot=<0 0 0 1>}
 altskin_cape_back {parent=pelvis, loc=<-26.72 -17.115 0>, rot=<0 0 0 1>}
 deform_glute_base_l {parent=pelvis, loc=<-8.8064 -8.8222 -8.4287>, rot=<0.0275 0.0501 -0.4472 -0.8926>}
 deform_glute_l {parent=deform_glute_base_l, loc=<0 0 0>, rot=<0 0 0 1>}
 deform_glute_base_r {parent=pelvis, loc=<-8.8062 -8.8222 8.4287>, rot=<0.0501 -0.0275 0.8926 -0.4472>}
 deform_glute_r {parent=deform_glute_base_r, loc=<0 0 0>, rot=<0 0 0 1>}
 dyn_skul_base {parent=pelvis, loc=<2.0777 5.2026 12.3957>, rot=<0.5944 0.3179 0.6302 0.3853>}
 dyn_skul {parent=dyn_skul_base, loc=<0 0 0>, rot=<0 0 0 1>}
 dyn_jaw {parent=dyn_skul_base, loc=<0 0 0>, rot=<-0.0405 0.0238 -0.2312 -0.9718>}
 dyn_garment_base_l {parent=pelvis, loc=<2.5484 8.8595 -4.105>, rot=<0.0043 0.0566 0.9981 0.0252>}
 dyn_garment_l_01 {parent=dyn_garment_base_l, loc=<13.2136 0 0>, rot=<-0.0002 -0.0052 0.0399 -0.9992>}
 dyn_garment_base_r {parent=pelvis, loc=<2.675 8.725 4.012>, rot=<0.0179 0.0554 0.998 0.026>}
 dyn_garment_r_01 {parent=dyn_garment_base_r, loc=<13.2136 0 0>, rot=<-0.0002 -0.0052 0.0399 -0.9992>}
 floor_joint {parent=root, loc=<0 0 0>, rot=<0 0 0 1>}
 attach {parent=root, loc=<0 0 0>, rot=<0 0 0 1>}
 ik_hand_root {parent=root, loc=<0 0 0>, rot=<0 0 0 1>}
 ik_hand_gun {parent=ik_hand_root, loc=<-51.9727 4.0883 102.6935>, rot=<0.6084 0.1793 0.4613 -0.6205>}
 ik_hand_r {parent=ik_hand_gun, loc=<0 0 0>, rot=<0 0 0 1>}
 ik_hand_l {parent=ik_hand_gun, loc=<53.0285 82.1781 35.2051>, rot=<-0.5102 -0.7906 -0.3387 0>}
 ik_foot_root {parent=root, loc=<0 0 0>, rot=<0 0 0 1>}
 ik_foot_l {parent=ik_foot_root, loc=<15.2608 -4.4176 9.4432>, rot=<0.0357 0.7045 0.0284 -0.7082>}
 ik_foot_r {parent=ik_foot_root, loc=<-15.2608 -4.4176 9.4432>, rot=<0.7082 0.0284 -0.7045 0.0357>}
 attach_fp {parent=root, loc=<0 0 0>, rot=<0 0 0 1>}
 camera_root {parent=root, loc=<0 0 0>, rot=<0 0 0 1>}
 camera {parent=camera_root, loc=<0 0 0>, rot=<0 0 0 1>}
 attach_emote_01 {parent=root, loc=<0 0 0>, rot=<0 0 0 1>}
 attach_emote_02 {parent=root, loc=<0 0 0>, rot=<0 0 0 1>}
 attach_emote_03 {parent=root, loc=<0 0 0>, rot=<0 0 0 1>}
 warp_point_01 {parent=root, loc=<0 0 0>, rot=<0 0 0 1>}
 warp_point_02 {parent=root, loc=<0 0 0>, rot=<0 0 0 1>}
 warp_point_03 {parent=root, loc=<0 0 0>, rot=<0 0 0 1>}
]
variables [
 Is jump {type=bool, default=False}
 Is sprinting crouching {type=bool, default=False}
 L thigh veryForward {type=double, default=0}
 Moving {type=bool, default=False}
 R thigh veryForward {type=double, default=0}
 RBAN enabled {type=bool, default=False}
 Value {type=FQuat, default=(X=0,Y=0,Z=0,W=1)}
 iscrouch {type=bool, default=False}
]
