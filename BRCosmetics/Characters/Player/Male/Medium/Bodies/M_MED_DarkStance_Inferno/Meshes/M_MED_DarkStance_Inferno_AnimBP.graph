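# Post-process style Anim Blueprint for the M_MED_DarkStance_Inferno player body cosmetic, adding secondary motion to its cape and skull bones.
# It copies the attached parent mesh's pose, then applies additive bone offsets toggled by movement states like crouch, slide, skydive and sprint.
# A Control Rig, thigh constraints on the skull and a pelvis-space RigidBody simulation follow.
# Chained per-bone AnimDynamics springs then drive six cape strands.
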AnimBlueprint /BRCosmetics/Characters/Player/Male/Medium/Bodies/M_MED_DarkStance_Inferno/Meshes/M_MED_DarkStance_Inferno_AnimBP
target_skeleton=&/M_MED_DarkStance_Inferno_Skeleton
graphs [
 AnimGraph {nodes [
   n1 = Get bIsCrouching()
   n2 = "Transform (Modify) Bone - Bone: dyn_cape_s_l_base"!(ComponentPose, bAlphaBoolEnabled:bool=@n1, Translation="0.463,0.000,0.000", Rotation="-0.000,-18.520,0.000", Scale="1,1,1") -> (Pose) {
    alphaInputType=Bool
    rotationMode=BMM_Additive
    translationMode=BMM_Additive
    alphaBoolBlend {blendOption=Linear}
    boneToModify {boneName=dyn_cape_s_l_base}
    rotation {yaw=-18.52}
    translation {x=0.463}
   }
   "Transform (Modify) Bone - Bone: dyn_cape_s_r_base"!(ComponentPose=@n2, bAlphaBoolEnabled:bool=@n1, Translation="2.104,-2.980,0.000", Rotation="4.939,61.831,-4.128", Scale="1,1,1") -> (Pose) {
    alphaInputType=Bool
    rotationMode=BMM_Additive
    translationMode=BMM_Additive
    alphaBoolBlend {blendOption=Linear}
    boneToModify {boneName=dyn_cape_s_r_base}
    rotation=<rot 4.939 61.831 -4.128>
    translation=<2.104 -2.98>
   }
   n3 = Get kneeSlidingFix() -> (kneeSlidingFix:bool)
   n4 = "Transform (Modify) Bone - Bone: dyn_cape_b_r_base"!(
    ComponentPose=Local To Component(LocalPose=Copy Pose From Mesh() {bUseAttachedParent=true})
    bAlphaBoolEnabled:bool=@n3
    Translation="-3,0,5"
    Rotation="76.165,-0.000,0.000"
    Scale="1,1,1"
   ) -> (Pose) {
    alphaInputType=Bool
    rotationMode=BMM_Additive
    rotationSpace=BCS_BoneSpace
    translationMode=BMM_Additive
    translationSpace=BCS_BoneSpace
    alphaBoolBlend {blendInTime=0.5, blendOption=Linear, blendOutTime=0.5}
    boneToModify {boneName=dyn_cape_b_r_base}
    rotation=<rot 76.165 0 0>
    translation {x=-3, z=5}
   }
   n5 = "Transform (Modify) Bone - Bone: dyn_cape_b_l_base"!(ComponentPose=@n4, bAlphaBoolEnabled:bool=@n3, Translation="-3,0,5", Rotation="90,0,0", Scale="1,1,1") -> (Pose) {
    alphaInputType=Bool
    rotationMode=BMM_Additive
    rotationSpace=BCS_BoneSpace
    translationMode=BMM_Additive
    translationSpace=BCS_BoneSpace
    alphaBoolBlend {blendInTime=0.5, blendOption=Linear, blendOutTime=0.5}
    boneToModify {boneName=dyn_cape_b_l_base}
    rotation {pitch=90}
    translation {x=-3, z=5}
   }
   n6 = Get bIsSkyDiving()
   n7 = "Transform (Modify) Bone - Bone: dyn_cape_b_r_base"(ComponentPose=@n5, bAlphaBoolEnabled=@n6, Translation="10,0,0", Rotation="0,0,0", Scale="1,1,1") {
    alphaInputType=Bool
    rotationMode=BMM_Additive
    rotationSpace=BCS_BoneSpace
    translationMode=BMM_Additive
    translationSpace=BCS_BoneSpace
    alphaBoolBlend {blendInTime=0.5, blendOption=Linear, blendOutTime=0.5}
    boneToModify {boneName=dyn_cape_b_r_base}
    translation {x=10}
   }
   n8 = "Transform (Modify) Bone - Bone: dyn_cape_b_l_base"(ComponentPose=@n7, bAlphaBoolEnabled=@n6, Translation="10,0,0", Rotation="0,0,0", Scale="1,1,1") {
    alphaInputType=Bool
    rotationMode=BMM_Additive
    rotationSpace=BCS_BoneSpace
    translationMode=BMM_Additive
    translationSpace=BCS_BoneSpace
    alphaBoolBlend {blendInTime=0.5, blendOption=Linear, blendOutTime=0.5}
    boneToModify {boneName=dyn_cape_b_l_base}
    translation {x=10}
   }
   n9 = "Transform (Modify) Bone - Bone: dyn_cape_s_l_base"(ComponentPose=@n8, bAlphaBoolEnabled=@n6, Translation="10,0,0", Rotation="0,0,0", Scale="1,1,1") {
    alphaInputType=Bool
    rotationMode=BMM_Additive
    rotationSpace=BCS_BoneSpace
    translationMode=BMM_Additive
    translationSpace=BCS_BoneSpace
    alphaBoolBlend {blendInTime=0.5, blendOption=Linear, blendOutTime=0.5}
    boneToModify {boneName=dyn_cape_s_l_base}
    translation {x=10}
   }
   n10 = "Transform (Modify) Bone - Bone: dyn_cape_s_r_base"(ComponentPose=@n9, bAlphaBoolEnabled=@n6, Translation="10,0,0", Rotation="0,0,0", Scale="1,1,1") {
    alphaInputType=Bool
    rotationMode=BMM_Additive
    rotationSpace=BCS_BoneSpace
    translationMode=BMM_Additive
    translationSpace=BCS_BoneSpace
    alphaBoolBlend {blendInTime=0.5, blendOption=Linear, blendOutTime=0.5}
    boneToModify {boneName=dyn_cape_s_r_base}
    translation {x=10}
   }
   n11 = "Transform (Modify) Bone - Bone: dyn_skul_base"!(ComponentPose=@n10, bAlphaBoolEnabled:bool=Get bIsDBNO(), Translation="0,-2,0.5", Rotation="0,0,0", Scale="1,1,1") -> (Pose) {
    alphaInputType=Bool
    rotationMode=BMM_Additive
    rotationSpace=BCS_BoneSpace
    translationMode=BMM_Additive
    translationSpace=BCS_BoneSpace
    alphaBoolBlend {blendInTime=0.5, blendOption=Linear, blendOutTime=0.5}
    boneToModify {boneName=dyn_skul_base}
    translation {y=-2, z=0.5}
   }
   n12 = Control Rig!(
    Source=Component To Local(ComponentPose=@n11)
    Alpha
    Is sprinting crouching:bool=Get bIsCrouchSprinting()
    iscrouch:bool=Get bIsCrouching()
    Is jump:bool=Get Jump RBAN() -> (Jump RBAN:bool)
   ) -> (Pose) {
    alphaBoolBlend {blendInTime=0.5, blendOption=Linear, blendOutTime=0.5}
    controlRigAssetReference {blueprintRigClass=&/M_MED_DarkStance_Inferno_CtrlRig.M_MED_DarkStance_Inferno_CtrlRig_C}
   }
   n13 = Anim Dynamics - Bone: dyn_skul(ComponentPose=Local To Component(LocalPose=@n12), Alpha=1, ExternalForce="0,0,0") {
    angularDampingOverride=0.9999
    angularSpringConstant=500
    bAngularSpring=true
    bGravityOverrideInSimSpace=true
    bOverrideAngularDamping=true
    simulationSpace=BoneRelative
    alphaBoolBlend {blendOption=Linear}
    boundBone {boneName=dyn_skul}
    componentAppliedLinearAccClamp=<3500 3500 3500>
    componentLinearAccScale=<1.2 1.2 1.2>
    componentLinearVelScale=<1.2 1.2 1.2>
    physicsBodyDefinitions [{
      sphereCollisionRadius=0
      boundBone {boneName=dyn_skul}
      boxExtents=<15 20 20>
      constraintSetup {angularLimitsMax=<10 0 5>, angularLimitsMin=<-10 0 -5>, angularTarget=<0 0 0>}
      localJointOffset=<4 0 0>
     }]
    relativeSpaceBone {boneName=root}
   }
   n14 = "Transform (Modify) Bone - Bone: dyn_cape_f_r_base"!(ComponentPose=@n13, bAlphaBoolEnabled:bool=True, Translation="0,0,5", Rotation="8.183,67.281,-9.271", Scale="1,1,1") -> (Pose) {
    alphaInputType=Bool
    rotationMode=BMM_Additive
    rotationSpace=BCS_BoneSpace
    translationMode=BMM_Additive
    translationSpace=BCS_BoneSpace
    alphaBoolBlend {blendInTime=0.5, blendOption=Linear, blendOutTime=0.5}
    boneToModify {boneName=dyn_cape_f_r_base}
    rotation=<rot 8.183 67.281 -9.271>
    translation {z=5}
   }
   n15 = "Transform (Modify) Bone - Bone: dyn_cape_f_l_base"!(ComponentPose=@n14, bAlphaBoolEnabled:bool=True, Translation="0,0,0", Rotation="43.420,-104.061,-89.749", Scale="1,1,1") -> (Pose) {
    alphaInputType=Bool
    rotationMode=BMM_Additive
    rotationSpace=BCS_BoneSpace
    translationMode=BMM_Additive
    translationSpace=BCS_BoneSpace
    alphaBoolBlend {blendInTime=0.5, blendOption=Linear, blendOutTime=0.5}
    boneToModify {boneName=dyn_cape_f_l_base}
    rotation=<rot 43.42 -104.061 -89.749>
   }
   n16 = "Transform (Modify) Bone - Bone: dyn_cape_s_l_base"(ComponentPose=@n15, bAlphaBoolEnabled=True, Translation="-3.666,-4.158,-4.657", Rotation="-13.698,-27.854,-31.532", Scale="1,1,1") {
    alphaInputType=Bool
    rotationMode=BMM_Additive
    rotationSpace=BCS_BoneSpace
    translationMode=BMM_Additive
    translationSpace=BCS_BoneSpace
    alphaBoolBlend {blendInTime=0.5, blendOption=Linear, blendOutTime=0.5}
    boneToModify {boneName=dyn_cape_s_l_base}
    rotation=<rot -13.698 -27.854 -31.532>
    translation=<-3.666 -4.158 -4.657>
   }
   n17 = "Transform (Modify) Bone - Bone: dyn_cape_s_r_base"(ComponentPose=@n16, bAlphaBoolEnabled=True, Translation="-6.179,-3.056,2.914", Rotation="-16.957,45.777,-0.559", Scale="1,1,1") {
    alphaInputType=Bool
    rotationMode=BMM_Additive
    rotationSpace=BCS_BoneSpace
    translationMode=BMM_Additive
    translationSpace=BCS_BoneSpace
    alphaBoolBlend {blendInTime=0.5, blendOption=Linear, blendOutTime=0.5}
    boneToModify {boneName=dyn_cape_s_r_base}
    rotation=<rot -16.957 45.777 -0.559>
    translation=<-6.179 -3.056 2.914>
   }
   n18 = "Transform (Modify) Bone - Bone: dyn_cape_b_l_base"(ComponentPose=@n17, bAlphaBoolEnabled=True, Translation="-5.045,0.050,5.336", Rotation="-17.982,-0.402,0.031", Scale="1,1,1") {
    alphaInputType=Bool
    rotationMode=BMM_Additive
    rotationSpace=BCS_BoneSpace
    translationMode=BMM_Additive
    translationSpace=BCS_BoneSpace
    alphaBoolBlend {blendInTime=0.5, blendOption=Linear, blendOutTime=0.5}
    boneToModify {boneName=dyn_cape_b_l_base}
    rotation=<rot -17.9821 -0.402 0.031>
    translation=<-5.045 0.05 5.336>
   }
   n19 = "Transform (Modify) Bone - Bone: dyn_cape_b_r_base"(ComponentPose=@n18, bAlphaBoolEnabled=True, Translation="-2.384,0.000,5.316", Rotation="2.075,0.001,0.007", Scale="1,1,1") {
    alphaInputType=Bool
    rotationMode=BMM_Additive
    rotationSpace=BCS_BoneSpace
    translationMode=BMM_Additive
    translationSpace=BCS_BoneSpace
    alphaBoolBlend {blendInTime=0.5, blendOption=Linear, blendOutTime=0.5}
    boneToModify {boneName=dyn_cape_b_r_base}
    rotation=<rot 2.075 0.001 0.007>
    translation {x=-2.384, z=5.316}
   }
   n20 = Get bIsBaseballSliding()
   n21 = "Transform (Modify) Bone - Bone: dyn_cape_f_r_base"(ComponentPose=@n19, bAlphaBoolEnabled=@n20, Translation="0,0,0", Rotation="37.987,113.214,29.093", Scale="1,1,1") {
    alphaInputType=Bool
    rotationMode=BMM_Additive
    rotationSpace=BCS_BoneSpace
    translationMode=BMM_Additive
    translationSpace=BCS_BoneSpace
    alphaBoolBlend {blendInTime=0.5, blendOption=Linear, blendOutTime=0.5}
    boneToModify {boneName=dyn_cape_f_r_base}
    rotation=<rot 37.987 113.214 29.093>
   }
   n22 = "Transform (Modify) Bone - Bone: dyn_cape_s_r_base"(ComponentPose=@n21, bAlphaBoolEnabled=@n20, Translation="-6.046,-2.139,1.199", Rotation="59.917,121.230,80.446", Scale="1,1,1") {
    alphaInputType=Bool
    rotationMode=BMM_Additive
    rotationSpace=BCS_BoneSpace
    translationMode=BMM_Additive
    translationSpace=BCS_BoneSpace
    alphaBoolBlend {blendInTime=0.5, blendOption=Linear, blendOutTime=0.5}
    boneToModify {boneName=dyn_cape_s_r_base}
    rotation=<rot 59.917 121.23 80.446>
    translation=<-6.046 -2.139 1.199>
   }
   n23 = "Transform (Modify) Bone - Bone: dyn_cape_f_l_base"(ComponentPose=@n22, bAlphaBoolEnabled=@n20, Translation="0,0,0", Rotation="-19.775,-8.498,-52.670", Scale="1,1,1") {
    alphaInputType=Bool
    rotationMode=BMM_Additive
    rotationSpace=BCS_BoneSpace
    translationMode=BMM_Additive
    translationSpace=BCS_BoneSpace
    alphaBoolBlend {blendInTime=0.5, blendOption=Linear, blendOutTime=0.5}
    boneToModify {boneName=dyn_cape_f_l_base}
    rotation=<rot -19.775 -8.498 -52.67>
   }
   n24 = "Transform (Modify) Bone - Bone: dyn_cape_s_l_base"(ComponentPose=@n23, bAlphaBoolEnabled=@n20, Translation="0,0,0", Rotation="74.386,-155.607,-124.952", Scale="1,1,1") {
    alphaInputType=Bool
    rotationMode=BMM_Additive
    rotationSpace=BCS_BoneSpace
    translationMode=BMM_Additive
    translationSpace=BCS_BoneSpace
    alphaBoolBlend {blendInTime=0.5, blendOption=Linear, blendOutTime=0.5}
    boneToModify {boneName=dyn_cape_s_l_base}
    rotation=<rot 74.386 -155.607 -124.952>
   }
   n25 = "Transform (Modify) Bone - Bone: dyn_cape_b_l_base"(ComponentPose=@n24, bAlphaBoolEnabled=@n20, Translation="0,0,0", Rotation="78.312,-180.000,-154.469", Scale="1,1,1") {
    alphaInputType=Bool
    rotationMode=BMM_Additive
    rotationSpace=BCS_BoneSpace
    translationMode=BMM_Additive
    translationSpace=BCS_BoneSpace
    alphaBoolBlend {blendInTime=0.5, blendOption=Linear, blendOutTime=0.5}
    boneToModify {boneName=dyn_cape_b_l_base}
    rotation=<rot 78.312 -180 -154.469>
   }
   n26 = "Transform (Modify) Bone - Bone: dyn_cape_b_r_base"(ComponentPose=@n25, bAlphaBoolEnabled=@n20, Translation="0,0,0", Rotation="61.024,-180.000,180.000", Scale="1,1,1") {
    alphaInputType=Bool
    rotationMode=BMM_Additive
    rotationSpace=BCS_BoneSpace
    translationMode=BMM_Additive
    translationSpace=BCS_BoneSpace
    alphaBoolBlend {blendInTime=0.5, blendOption=Linear, blendOutTime=0.5}
    boneToModify {boneName=dyn_cape_b_r_base}
    rotation=<rot 61.024 -180 180>
   }
   n27 = Get bIsSprinting()
   n28 = "Transform (Modify) Bone - Bone: dyn_cape_s_l_base"(ComponentPose=@n26, bAlphaBoolEnabled=@n27, Translation="0,0,0", Rotation="-0.000,0.000,8.398", Scale="1,1,1") {
    alphaInputType=Bool
    rotationMode=BMM_Additive
    translationMode=BMM_Additive
    alphaBoolBlend {blendOption=Linear}
    boneToModify {boneName=dyn_cape_s_l_base}
    rotation {roll=8.398}
   }
   n29 = "Transform (Modify) Bone - Bone: dyn_cape_s_r_base"(ComponentPose=@n28, bAlphaBoolEnabled=@n27, Translation="0,0,0", Rotation="-2.159,-0.288,13.131", Scale="1,1,1") {
    alphaInputType=Bool
    rotationMode=BMM_Additive
    translationMode=BMM_Additive
    alphaBoolBlend {blendOption=Linear}
    boneToModify {boneName=dyn_cape_s_r_base}
    rotation=<rot -2.159 -0.288 13.131>
   }
   n30 = "Transform (Modify) Bone - Bone: dyn_cape_f_l_base"(ComponentPose=@n29, bAlphaBoolEnabled=@n1, Translation="0.869,-0.489,-1.208", Rotation="21.540,-0.000,-22.041", Scale="1,1,1") {
    alphaInputType=Bool
    rotationMode=BMM_Additive
    translationMode=BMM_Additive
    alphaBoolBlend {blendOption=Linear}
    boneToModify {boneName=dyn_cape_f_l_base}
    rotation=<rot 21.54 0 -22.041>
    translation=<0.869 -0.489 -1.208>
   }
   n31 = "Transform (Modify) Bone - Bone: dyn_cape_f_r_base"(ComponentPose=@n30, bAlphaBoolEnabled=@n1, Translation="0,0,0", Rotation="-13.780,-11.987,-6.732", Scale="1,1,1") {
    alphaInputType=Bool
    rotationMode=BMM_Additive
    rotationSpace=BCS_BoneSpace
    translationMode=BMM_Additive
    translationSpace=BCS_BoneSpace
    alphaBoolBlend {blendInTime=0.5, blendOption=Linear, blendOutTime=0.5}
    boneToModify {boneName=dyn_cape_f_r_base}
    rotation=<rot -13.78 -11.987 -6.732>
   }
   n32 = "Transform (Modify) Bone - Bone: dyn_cape_b_l_base"(ComponentPose=@n31, bAlphaBoolEnabled=@n1, Translation="0,0,5", Rotation="1.358,-10.567,7.241", Scale="1,1,1") {
    alphaInputType=Bool
    rotationMode=BMM_Additive
    translationMode=BMM_Additive
    alphaBoolBlend {blendOption=Linear}
    boneToModify {boneName=dyn_cape_b_l_base}
    rotation=<rot 1.358 -10.567 7.241>
    translation {z=5}
   }
   n33 = "Transform (Modify) Bone - Bone: dyn_cape_b_r_base"(ComponentPose=@n32, bAlphaBoolEnabled=@n1, Translation="-5.704,0,0", Rotation="0,0,0.006", Scale="1,1,1") {
    alphaInputType=Bool
    rotationMode=BMM_Additive
    rotationSpace=BCS_BoneSpace
    translationMode=BMM_Additive
    translationSpace=BCS_BoneSpace
    alphaBoolBlend {blendInTime=0.5, blendOption=Linear, blendOutTime=0.5}
    boneToModify {boneName=dyn_cape_b_r_base}
    rotation {roll=0.006}
    translation {x=-5.704}
   }
   n34 = "Constraint - dyn_skul_base (R)"!(ComponentPose=@n33, bAlphaBoolEnabled:bool=Get bIsCrouching(), ConstraintWeights_0=1) -> (Pose) {
    alphaInputType=Bool
    alphaBoolBlend {blendOption=Linear}
    boneToModify {boneName=dyn_skul_base}
    constraintSetup [{transformType=Rotation, targetBone {boneName=thigh_r}}]
    constraintWeights [1]
   }
   n35 = "Constraint - dyn_skul_base (R)"(ComponentPose=@n34, bAlphaBoolEnabled=Get bIsParachuteOpened(), ConstraintWeights_0=1) {
    alphaInputType=Bool
    alphaBoolBlend {blendOption=Linear}
    boneToModify {boneName=dyn_skul_base}
    constraintSetup [{transformType=Rotation, targetBone {boneName=thigh_r}}]
    constraintWeights [1]
   }
   n36 = RigidBody!(
    ComponentPose=@n35
    bAlphaBoolEnabled:bool
    OverrideWorldGravity:Vector=Get Gravity Override()
    ExternalForce
    SimSpaceSettings:Sim Space Settings Structure=Get Sim Space Settings() -> (Sim Space Settings:Sim Space Settings Structure)
   ) -> (Pose) {
    alphaInputType=Bool
    bClampLinearTranslationLimitToRefPose=true
    bOverrideWorldGravity=true
    overridePhysicsAsset=&/M_MED_DarkStance_Inferno_Cape_Physics
    simulationSpace=BaseBoneSpace
    alphaBoolBlend {blendOption=Linear}
    baseBoneRef {boneName=pelvis}
    componentAppliedLinearAccClamp=<3000 3000 1000>
    componentLinearAccScale=<0.5 0.4 0.4>
    componentLinearVelScale=<0.5 0.4 0.4>
    overrideWorldGravity {z=-920}
   }
   n37 = Get bIsSkyDiving()
   n38 = "Transform (Modify) Bone - Bone: dyn_cape_f_l_base"(ComponentPose=@n36, bAlphaBoolEnabled=@n37, Translation="-4.069,0.167,-1.011", Rotation="0,0,0.017", Scale="1,1,1") {
    alphaInputType=Bool
    rotationMode=BMM_Additive
    rotationSpace=BCS_BoneSpace
    translationMode=BMM_Additive
    translationSpace=BCS_BoneSpace
    alphaBoolBlend {blendInTime=0.5, blendOption=Linear, blendOutTime=0.5}
    boneToModify {boneName=dyn_cape_f_l_base}
    rotation {roll=0.017}
    translation=<-4.069 0.167 -1.011>
   }
   n39 = "Transform (Modify) Bone - Bone: dyn_cape_f_r_base"(ComponentPose=@n38, bAlphaBoolEnabled=@n37, Translation="-0.075,0.383,-0.515", Rotation="0,0,0", Scale="1,1,1") {
    alphaInputType=Bool
    rotationMode=BMM_Additive
    rotationSpace=BCS_BoneSpace
    translationMode=BMM_Additive
    translationSpace=BCS_BoneSpace
    alphaBoolBlend {blendInTime=0.5, blendOption=Linear, blendOutTime=0.5}
    boneToModify {boneName=dyn_cape_f_r_base}
    translation=<-0.075 0.383 -0.515>
   }
   n40 = "Transform (Modify) Bone - Bone: dyn_cape_f_l_01"!(ComponentPose=@n39, Alpha=1) -> (Pose) {
    rotationMode=BMM_Additive
    rotationSpace=BCS_BoneSpace
    alphaBoolBlend {blendOption=Linear}
    boneToModify {boneName=dyn_cape_f_l_01}
    rotation {pitch=-10, yaw=-10}
   }
   n41 = "Transform (Modify) Bone - Bone: dyn_cape_f_r_01"!(ComponentPose=@n40, Alpha=1) -> (Pose) {
    rotationMode=BMM_Additive
    rotationSpace=BCS_BoneSpace
    alphaBoolBlend {blendOption=Linear}
    boneToModify {boneName=dyn_cape_f_r_01}
    rotation {pitch=-10, yaw=-10}
   }
   n42 = "Transform (Modify) Bone - Bone: dyn_cape_f_l_02"!(ComponentPose=@n41, Alpha=1) -> (Pose) {
    rotationMode=BMM_Additive
    rotationSpace=BCS_BoneSpace
    alphaBoolBlend {blendOption=Linear}
    boneToModify {boneName=dyn_cape_f_l_02}
    rotation {pitch=-10, yaw=-10}
   }
   n43 = "Transform (Modify) Bone - Bone: dyn_cape_f_r_02"!(ComponentPose=@n42, Alpha=1) -> (Pose) {
    rotationMode=BMM_Additive
    rotationSpace=BCS_BoneSpace
    alphaBoolBlend {blendOption=Linear}
    boneToModify {boneName=dyn_cape_f_r_02}
    rotation {pitch=-10, yaw=-10}
   }
   n44 = "Transform (Modify) Bone - Bone: dyn_cape_f_l_03"!(ComponentPose=@n43, Alpha=1) -> (Pose) {
    rotationMode=BMM_Additive
    rotationSpace=BCS_BoneSpace
    alphaBoolBlend {blendOption=Linear}
    boneToModify {boneName=dyn_cape_f_l_03}
    rotation {pitch=-10, yaw=-10}
   }
   n45 = "Transform (Modify) Bone - Bone: dyn_cape_f_r_03"!(ComponentPose=@n44, Alpha=1) -> (Pose) {
    rotationMode=BMM_Additive
    rotationSpace=BCS_BoneSpace
    alphaBoolBlend {blendOption=Linear}
    boneToModify {boneName=dyn_cape_f_r_03}
    rotation {pitch=-10, yaw=-10}
   }
   n46 = Anim Dynamics - Bone: dyn_cape_f_l_01!(SimulationSpace:AnimPhysSimSpaceType Enum, ComponentPose=@n45, Alpha=1, GravityOverride:Vector="0,0,-980", ExternalForce="0,0,0") -> (Pose) {
    angularDampingOverride=0.99
    angularSpringConstant=50
    bAngularSpring=true
    bOverrideAngularDamping=true
    bUseGravityOverride=true
    gravityScale=0.35
    simulationSpace=RootRelative
    alphaBoolBlend {blendOption=Linear}
    boundBone {boneName=dyn_cape_f_l_01}
    componentLinearAccScale=<0.4 0.4 0.4>
    componentLinearVelScale=<0.4 0.4 0.4>
    gravityOverride {z=-980}
    physicsBodyDefinitions [{
      boundBone {boneName=dyn_cape_f_l_01}
      boxExtents=<16 16 16>
      constraintSetup {angularLimitsMax=<0 0 1>, angularLimitsMin=<0 -15 -5>, angularTarget=<1 0 0>}
      localJointOffset=<2 0 0>
     }]
   }
   n47 = Anim Dynamics - Bone: dyn_cape_f_l_02!(SimulationSpace:AnimPhysSimSpaceType Enum, ComponentPose=@n46, Alpha=1, GravityOverride:Vector="0,0,-1000", ExternalForce="0,0,0") -> (Pose) {
    angularDampingOverride=0.999
    angularSpringConstant=50
    bAngularSpring=true
    bOverrideAngularDamping=true
    bUseGravityOverride=true
    gravityScale=0.35
    simulationSpace=RootRelative
    alphaBoolBlend {blendOption=Linear}
    boundBone {boneName=dyn_cape_f_l_02}
    componentLinearAccScale=<0.4 0.4 0.4>
    componentLinearVelScale=<0.4 0.4 0.4>
    gravityOverride {z=-1000}
    physicsBodyDefinitions [{
      boundBone {boneName=dyn_cape_f_l_02}
      boxExtents=<16 16 16>
      constraintSetup {angularLimitsMax=<0 0 1>, angularLimitsMin=<0 -15 -5>, angularTarget=<1 0 0>}
      localJointOffset=<2 0 0>
     }]
   }
   n48 = Anim Dynamics - Bone: dyn_cape_f_l_03!(SimulationSpace:AnimPhysSimSpaceType Enum, ComponentPose=@n47, Alpha=1, GravityOverride:Vector="0,0,-1000", ExternalForce="0,0,0") -> (Pose) {
    angularDampingOverride=0.999
    angularSpringConstant=50
    bAngularSpring=true
    bOverrideAngularDamping=true
    bUseGravityOverride=true
    gravityScale=0.35
    simulationSpace=RootRelative
    alphaBoolBlend {blendOption=Linear}
    boundBone {boneName=dyn_cape_f_l_03}
    componentLinearAccScale=<0.4 0.4 0.4>
    componentLinearVelScale=<0.4 0.4 0.4>
    gravityOverride {z=-1000}
    physicsBodyDefinitions [{
      boundBone {boneName=dyn_cape_f_l_03}
      boxExtents=<16 16 16>
      constraintSetup {angularLimitsMax=<0 0 1>, angularLimitsMin=<0 -15 -5>, angularTarget=<1 0 0>}
      localJointOffset=<2 0 0>
     }]
   }
   n49 = Anim Dynamics - Bone: dyn_cape_f_l_04!(SimulationSpace:AnimPhysSimSpaceType Enum, ComponentPose=@n48, Alpha=1, GravityOverride:Vector="0,0,-1000", ExternalForce="0,0,0") -> (Pose) {
    angularDampingOverride=0.999
    angularSpringConstant=50
    bAngularSpring=true
    bOverrideAngularDamping=true
    bUseGravityOverride=true
    gravityScale=0.35
    simulationSpace=RootRelative
    alphaBoolBlend {blendOption=Linear}
    boundBone {boneName=dyn_cape_f_l_04}
    componentLinearAccScale=<0.4 0.4 0.4>
    componentLinearVelScale=<0.4 0.4 0.4>
    gravityOverride {z=-1000}
    physicsBodyDefinitions [{
      boundBone {boneName=dyn_cape_f_l_04}
      boxExtents=<16 16 16>
      constraintSetup {angularLimitsMax=<0 0 1>, angularLimitsMin=<0 -15 -5>, angularTarget=<1 0 0>}
      localJointOffset=<2 0 0>
     }]
   }
   n50 = Anim Dynamics - Bone: dyn_cape_s_l_01!(SimulationSpace:AnimPhysSimSpaceType Enum, ComponentPose=@n49, Alpha=1, GravityOverride:Vector="0,0,-980", ExternalForce="0,0,0") -> (Pose) {
    angularDampingOverride=0.999
    angularSpringConstant=50
    bAngularSpring=true
    bUseGravityOverride=true
    gravityScale=0.35
    simulationSpace=RootRelative
    alphaBoolBlend {blendOption=Linear}
    boundBone {boneName=dyn_cape_s_l_01}
    componentLinearAccScale=<0.4 0.4 0.4>
    componentLinearVelScale=<0.4 0.4 0.4>
    gravityOverride {z=-980}
    physicsBodyDefinitions [{
      boundBone {boneName=dyn_cape_s_l_01}
      boxExtents=<16 16 16>
      constraintSetup {angularLimitsMax=<5 0 1>, angularLimitsMin=<-5 -15 -5>, angularTarget=<1 0 0>}
      localJointOffset=<2 0 0>
     }]
   }
   n51 = Anim Dynamics - Bone: dyn_cape_s_l_02!(SimulationSpace:AnimPhysSimSpaceType Enum, ComponentPose=@n50, Alpha=1, GravityOverride:Vector="0,0,-1000", ExternalForce="0,0,0") -> (Pose) {
    angularDampingOverride=0.999
    angularSpringConstant=50
    bAngularSpring=true
    bOverrideAngularDamping=true
    bUseGravityOverride=true
    gravityScale=0.35
    simulationSpace=RootRelative
    alphaBoolBlend {blendOption=Linear}
    boundBone {boneName=dyn_cape_s_l_02}
    componentLinearAccScale=<0.4 0.4 0.4>
    componentLinearVelScale=<0.4 0.4 0.4>
    gravityOverride {z=-1000}
    physicsBodyDefinitions [{
      boundBone {boneName=dyn_cape_s_l_02}
      boxExtents=<16 16 16>
      constraintSetup {angularLimitsMax=<5 0 1>, angularLimitsMin=<-5 -15 -5>, angularTarget=<1 0 0>}
      localJointOffset=<2 0 0>
     }]
   }
   n52 = Anim Dynamics - Bone: dyn_cape_s_l_03!(SimulationSpace:AnimPhysSimSpaceType Enum, ComponentPose=@n51, Alpha=1, GravityOverride:Vector="0,0,-1000", ExternalForce="0,0,0") -> (Pose) {
    angularDampingOverride=0.999
    angularSpringConstant=50
    bAngularSpring=true
    bOverrideAngularDamping=true
    bUseGravityOverride=true
    gravityScale=0.35
    simulationSpace=RootRelative
    alphaBoolBlend {blendOption=Linear}
    boundBone {boneName=dyn_cape_s_l_03}
    componentLinearAccScale=<0.4 0.4 0.4>
    componentLinearVelScale=<0.4 0.4 0.4>
    gravityOverride {z=-1000}
    physicsBodyDefinitions [{
      boundBone {boneName=dyn_cape_s_l_03}
      boxExtents=<16 16 16>
      constraintSetup {angularLimitsMax=<5 0 1>, angularLimitsMin=<-5 -15 -5>, angularTarget=<1 0 0>}
      localJointOffset=<2 0 0>
     }]
   }
   n53 = Anim Dynamics - Bone: dyn_cape_s_l_04!(SimulationSpace:AnimPhysSimSpaceType Enum, ComponentPose=@n52, Alpha=1, GravityOverride:Vector="0,0,-1000", ExternalForce="0,0,0") -> (Pose) {
    angularDampingOverride=0.999
    angularSpringConstant=50
    bAngularSpring=true
    bOverrideAngularDamping=true
    bUseGravityOverride=true
    gravityScale=0.35
    simulationSpace=RootRelative
    alphaBoolBlend {blendOption=Linear}
    boundBone {boneName=dyn_cape_s_l_04}
    componentLinearAccScale=<0.4 0.4 0.4>
    componentLinearVelScale=<0.4 0.4 0.4>
    gravityOverride {z=-1000}
    physicsBodyDefinitions [{
      boundBone {boneName=dyn_cape_s_l_04}
      boxExtents=<16 16 16>
      constraintSetup {angularLimitsMax=<5 0 1>, angularLimitsMin=<-5 -15 -5>, angularTarget=<1 0 0>}
      localJointOffset=<2 0 0>
     }]
   }
   n54 = Anim Dynamics - Bone: dyn_cape_f_r_01!(SimulationSpace:AnimPhysSimSpaceType Enum, ComponentPose=@n53, Alpha=1, GravityOverride:Vector="0,0,-980", ExternalForce="0,0,0") -> (Pose) {
    angularDampingOverride=0.999
    angularSpringConstant=50
    bAngularSpring=true
    bUseGravityOverride=true
    gravityScale=0.35
    simulationSpace=RootRelative
    alphaBoolBlend {blendOption=Linear}
    boundBone {boneName=dyn_cape_f_r_01}
    componentLinearAccScale=<0.4 0.4 0.4>
    componentLinearVelScale=<0.4 0.4 0.4>
    gravityOverride {z=-980}
    physicsBodyDefinitions [{
      boundBone {boneName=dyn_cape_f_r_01}
      boxExtents=<16 16 16>
      constraintSetup {angularLimitsMax=<0 0 5>, angularLimitsMin=<0 -15 -1>, angularTarget=<1 0 0>}
      localJointOffset=<2 0 0>
     }]
   }
   n55 = Anim Dynamics - Bone: dyn_cape_f_r_02!(SimulationSpace:AnimPhysSimSpaceType Enum, ComponentPose=@n54, Alpha=1, GravityOverride:Vector="0,0,-1000", ExternalForce="0,0,0") -> (Pose) {
    angularDampingOverride=0.999
    angularSpringConstant=50
    bAngularSpring=true
    bOverrideAngularDamping=true
    bUseGravityOverride=true
    gravityScale=0.35
    simulationSpace=RootRelative
    alphaBoolBlend {blendOption=Linear}
    boundBone {boneName=dyn_cape_f_r_02}
    componentLinearAccScale=<0.4 0.4 0.4>
    componentLinearVelScale=<0.4 0.4 0.4>
    gravityOverride {z=-1000}
    physicsBodyDefinitions [{
      boundBone {boneName=dyn_cape_f_r_02}
      boxExtents=<16 16 16>
      constraintSetup {angularLimitsMax=<0 0 5>, angularLimitsMin=<0 -15 -1>, angularTarget=<1 0 0>}
      localJointOffset=<2 0 0>
     }]
   }
   n56 = Anim Dynamics - Bone: dyn_cape_f_r_03!(SimulationSpace:AnimPhysSimSpaceType Enum, ComponentPose=@n55, Alpha=1, GravityOverride:Vector="0,0,-1000", ExternalForce="0,0,0") -> (Pose) {
    angularDampingOverride=0.999
    angularSpringConstant=50
    bAngularSpring=true
    bOverrideAngularDamping=true
    bUseGravityOverride=true
    gravityScale=0.35
    simulationSpace=RootRelative
    alphaBoolBlend {blendOption=Linear}
    boundBone {boneName=dyn_cape_f_r_03}
    componentLinearAccScale=<0.4 0.4 0.4>
    componentLinearVelScale=<0.4 0.4 0.4>
    gravityOverride {z=-1000}
    physicsBodyDefinitions [{
      boundBone {boneName=dyn_cape_f_r_03}
      boxExtents=<16 16 16>
      constraintSetup {angularLimitsMax=<0 0 5>, angularLimitsMin=<0 -15 -1>, angularTarget=<1 0 0>}
      localJointOffset=<2 0 0>
     }]
   }
   n57 = Anim Dynamics - Bone: dyn_cape_f_r_04!(SimulationSpace:AnimPhysSimSpaceType Enum, ComponentPose=@n56, Alpha=1, GravityOverride:Vector="0,0,-1000", ExternalForce="0,0,0") -> (Pose) {
    angularDampingOverride=0.999
    angularSpringConstant=50
    bAngularSpring=true
    bOverrideAngularDamping=true
    bUseGravityOverride=true
    gravityScale=0.35
    simulationSpace=RootRelative
    alphaBoolBlend {blendOption=Linear}
    boundBone {boneName=dyn_cape_f_r_04}
    componentLinearAccScale=<0.4 0.4 0.4>
    componentLinearVelScale=<0.4 0.4 0.4>
    gravityOverride {z=-1000}
    physicsBodyDefinitions [{
      boundBone {boneName=dyn_cape_f_r_04}
      boxExtents=<16 16 16>
      constraintSetup {angularLimitsMax=<0 0 5>, angularLimitsMin=<0 -15 -1>, angularTarget=<1 0 0>}
      localJointOffset=<2 0 0>
     }]
   }
   n58 = Anim Dynamics - Bone: dyn_cape_s_r_01!(SimulationSpace:AnimPhysSimSpaceType Enum, ComponentPose=@n57, Alpha=1, GravityOverride:Vector="0,0,-980", ExternalForce="0,0,0") -> (Pose) {
    angularDampingOverride=0.999
    angularSpringConstant=50
    bAngularSpring=true
    bUseGravityOverride=true
    gravityScale=0.35
    simulationSpace=RootRelative
    alphaBoolBlend {blendOption=Linear}
    boundBone {boneName=dyn_cape_s_r_01}
    componentLinearAccScale=<0.4 0.4 0.4>
    componentLinearVelScale=<0.4 0.4 0.4>
    gravityOverride {z=-980}
    physicsBodyDefinitions [{
      boundBone {boneName=dyn_cape_s_r_01}
      boxExtents=<16 16 16>
      constraintSetup {angularLimitsMax=<5 0 5>, angularLimitsMin=<-5 -15 -1>, angularTarget=<1 0 0>}
      localJointOffset=<2 0 0>
     }]
   }
   n59 = Anim Dynamics - Bone: dyn_cape_s_r_02!(SimulationSpace:AnimPhysSimSpaceType Enum, ComponentPose=@n58, Alpha=1, GravityOverride:Vector="0,0,-1000", ExternalForce="0,0,0") -> (Pose) {
    angularDampingOverride=0.999
    angularSpringConstant=50
    bAngularSpring=true
    bOverrideAngularDamping=true
    bUseGravityOverride=true
    gravityScale=0.35
    simulationSpace=RootRelative
    alphaBoolBlend {blendOption=Linear}
    boundBone {boneName=dyn_cape_s_r_02}
    componentLinearAccScale=<0.4 0.4 0.4>
    componentLinearVelScale=<0.4 0.4 0.4>
    gravityOverride {z=-1000}
    physicsBodyDefinitions [{
      boundBone {boneName=dyn_cape_s_r_02}
      boxExtents=<16 16 16>
      constraintSetup {angularLimitsMax=<5 0 5>, angularLimitsMin=<-5 -15 -1>, angularTarget=<1 0 0>}
      localJointOffset=<2 0 0>
     }]
   }
   n60 = Anim Dynamics - Bone: dyn_cape_s_r_03!(SimulationSpace:AnimPhysSimSpaceType Enum, ComponentPose=@n59, Alpha=1, GravityOverride:Vector="0,0,-1000", ExternalForce="0,0,0") -> (Pose) {
    angularDampingOverride=0.999
    angularSpringConstant=50
    bAngularSpring=true
    bOverrideAngularDamping=true
    bUseGravityOverride=true
    gravityScale=0.35
    simulationSpace=RootRelative
    alphaBoolBlend {blendOption=Linear}
    boundBone {boneName=dyn_cape_s_r_03}
    componentLinearAccScale=<0.4 0.4 0.4>
    componentLinearVelScale=<0.4 0.4 0.4>
    gravityOverride {z=-1000}
    physicsBodyDefinitions [{
      boundBone {boneName=dyn_cape_s_r_03}
      boxExtents=<16 16 16>
      constraintSetup {angularLimitsMax=<5 0 5>, angularLimitsMin=<-5 -15 -1>, angularTarget=<1 0 0>}
      localJointOffset=<2 0 0>
     }]
   }
   n61 = Anim Dynamics - Bone: dyn_cape_s_r_04!(SimulationSpace:AnimPhysSimSpaceType Enum, ComponentPose=@n60, Alpha=1, GravityOverride:Vector="0,0,-1000", ExternalForce="0,0,0") -> (Pose) {
    angularDampingOverride=0.999
    angularSpringConstant=50
    bAngularSpring=true
    bOverrideAngularDamping=true
    bUseGravityOverride=true
    gravityScale=0.35
    simulationSpace=RootRelative
    alphaBoolBlend {blendOption=Linear}
    boundBone {boneName=dyn_cape_s_r_04}
    componentLinearAccScale=<0.4 0.4 0.4>
    componentLinearVelScale=<0.4 0.4 0.4>
    gravityOverride {z=-1000}
    physicsBodyDefinitions [{
      boundBone {boneName=dyn_cape_s_r_04}
      boxExtents=<16 16 16>
      constraintSetup {angularLimitsMax=<5 0 5>, angularLimitsMin=<-5 -15 -1>, angularTarget=<1 0 0>}
      localJointOffset=<2 0 0>
     }]
   }
   n62 = Anim Dynamics - Bone: dyn_cape_b_l_01!(SimulationSpace:AnimPhysSimSpaceType Enum, ComponentPose=@n61, Alpha=1, GravityOverride:Vector="0,0,-980", ExternalForce="0,0,0") -> (Pose) {
    angularDampingOverride=0.999
    angularSpringConstant=50
    bAngularSpring=true
    bUseGravityOverride=true
    gravityScale=0.35
    simulationSpace=RootRelative
    alphaBoolBlend {blendOption=Linear}
    boundBone {boneName=dyn_cape_b_l_01}
    componentLinearAccScale=<0.4 0.4 0.4>
    componentLinearVelScale=<0.4 0.4 0.4>
    gravityOverride {z=-980}
    physicsBodyDefinitions [{
      boundBone {boneName=dyn_cape_b_l_01}
      boxExtents=<16 16 16>
      constraintSetup {angularLimitsMax=<10 0 10>, angularLimitsMin=<-10 -15 -10>, angularTarget=<1 0 0>}
      localJointOffset=<2 0 0>
     }]
   }
   n63 = Anim Dynamics - Bone: dyn_cape_b_l_02!(SimulationSpace:AnimPhysSimSpaceType Enum, ComponentPose=@n62, Alpha=1, GravityOverride:Vector="0,0,-1000", ExternalForce="0,0,0") -> (Pose) {
    angularDampingOverride=0.999
    angularSpringConstant=50
    bAngularSpring=true
    bOverrideAngularDamping=true
    bUseGravityOverride=true
    gravityScale=0.35
    simulationSpace=RootRelative
    alphaBoolBlend {blendOption=Linear}
    boundBone {boneName=dyn_cape_b_l_02}
    componentLinearAccScale=<0.4 0.4 0.4>
    componentLinearVelScale=<0.4 0.4 0.4>
    gravityOverride {z=-1000}
    physicsBodyDefinitions [{
      boundBone {boneName=dyn_cape_b_l_02}
      boxExtents=<16 16 16>
      constraintSetup {angularLimitsMax=<10 0 10>, angularLimitsMin=<-10 -15 -10>, angularTarget=<1 0 0>}
      localJointOffset=<2 0 0>
     }]
   }
   n64 = Anim Dynamics - Bone: dyn_cape_b_l_03!(SimulationSpace:AnimPhysSimSpaceType Enum, ComponentPose=@n63, Alpha=1, GravityOverride:Vector="0,0,-1000", ExternalForce="0,0,0") -> (Pose) {
    angularDampingOverride=0.999
    angularSpringConstant=50
    bAngularSpring=true
    bOverrideAngularDamping=true
    bUseGravityOverride=true
    gravityScale=0.35
    simulationSpace=RootRelative
    alphaBoolBlend {blendOption=Linear}
    boundBone {boneName=dyn_cape_b_l_03}
    componentLinearAccScale=<0.4 0.4 0.4>
    componentLinearVelScale=<0.4 0.4 0.4>
    gravityOverride {z=-1000}
    physicsBodyDefinitions [{
      boundBone {boneName=dyn_cape_b_l_03}
      boxExtents=<16 16 16>
      constraintSetup {angularLimitsMax=<10 0 10>, angularLimitsMin=<-10 -15 -10>, angularTarget=<1 0 0>}
      localJointOffset=<2 0 0>
     }]
   }
   n65 = Anim Dynamics - Bone: dyn_cape_b_l_04!(SimulationSpace:AnimPhysSimSpaceType Enum, ComponentPose=@n64, Alpha=1, GravityOverride:Vector="0,0,-1000", ExternalForce="0,0,0") -> (Pose) {
    angularDampingOverride=0.999
    angularSpringConstant=50
    bAngularSpring=true
    bOverrideAngularDamping=true
    bUseGravityOverride=true
    gravityScale=0.35
    simulationSpace=RootRelative
    alphaBoolBlend {blendOption=Linear}
    boundBone {boneName=dyn_cape_b_l_04}
    componentLinearAccScale=<0.4 0.4 0.4>
    componentLinearVelScale=<0.4 0.4 0.4>
    gravityOverride {z=-1000}
    physicsBodyDefinitions [{
      boundBone {boneName=dyn_cape_b_l_04}
      boxExtents=<16 16 16>
      constraintSetup {angularLimitsMax=<10 0 10>, angularLimitsMin=<-10 -15 -10>, angularTarget=<1 0 0>}
      localJointOffset=<2 0 0>
     }]
   }
   n66 = Anim Dynamics - Bone: dyn_cape_b_r_01!(SimulationSpace:AnimPhysSimSpaceType Enum, ComponentPose=@n65, Alpha=1, GravityOverride:Vector="0,0,-980", ExternalForce="0,0,0") -> (Pose) {
    angularDampingOverride=0.999
    angularSpringConstant=50
    bAngularSpring=true
    bUseGravityOverride=true
    gravityScale=0.35
    simulationSpace=RootRelative
    alphaBoolBlend {blendOption=Linear}
    boundBone {boneName=dyn_cape_b_r_01}
    componentLinearAccScale=<0.4 0.4 0.4>
    componentLinearVelScale=<0.4 0.4 0.4>
    gravityOverride {z=-980}
    physicsBodyDefinitions [{
      boundBone {boneName=dyn_cape_b_r_01}
      boxExtents=<16 16 16>
      constraintSetup {angularLimitsMax=<10 0 10>, angularLimitsMin=<-10 -15 -10>, angularTarget=<1 0 0>}
      localJointOffset=<2 0 0>
     }]
   }
   n67 = Anim Dynamics - Bone: dyn_cape_b_r_02!(SimulationSpace:AnimPhysSimSpaceType Enum, ComponentPose=@n66, Alpha=1, GravityOverride:Vector="0,0,-1000", ExternalForce="0,0,0") -> (Pose) {
    angularDampingOverride=0.999
    angularSpringConstant=50
    bAngularSpring=true
    bOverrideAngularDamping=true
    bUseGravityOverride=true
    gravityScale=0.35
    simulationSpace=RootRelative
    alphaBoolBlend {blendOption=Linear}
    boundBone {boneName=dyn_cape_b_r_02}
    componentLinearAccScale=<0.4 0.4 0.4>
    componentLinearVelScale=<0.4 0.4 0.4>
    gravityOverride {z=-1000}
    physicsBodyDefinitions [{
      boundBone {boneName=dyn_cape_b_r_02}
      boxExtents=<16 16 16>
      constraintSetup {angularLimitsMax=<10 0 10>, angularLimitsMin=<-10 -15 -10>, angularTarget=<1 0 0>}
      localJointOffset=<2 0 0>
     }]
   }
   n68 = Anim Dynamics - Bone: dyn_cape_b_r_03!(SimulationSpace:AnimPhysSimSpaceType Enum, ComponentPose=@n67, Alpha=1, GravityOverride:Vector="0,0,-1000", ExternalForce="0,0,0") -> (Pose) {
    angularDampingOverride=0.999
    angularSpringConstant=50
    bAngularSpring=true
    bOverrideAngularDamping=true
    bUseGravityOverride=true
    gravityScale=0.35
    simulationSpace=RootRelative
    alphaBoolBlend {blendOption=Linear}
    boundBone {boneName=dyn_cape_b_r_03}
    componentLinearAccScale=<0.4 0.4 0.4>
    componentLinearVelScale=<0.4 0.4 0.4>
    gravityOverride {z=-1000}
    physicsBodyDefinitions [{
      boundBone {boneName=dyn_cape_b_r_03}
      boxExtents=<16 16 16>
      constraintSetup {angularLimitsMax=<10 0 10>, angularLimitsMin=<-10 -15 -10>, angularTarget=<1 0 0>}
      localJointOffset=<2 0 0>
     }]
   }
   n69 = Anim Dynamics - Bone: dyn_cape_b_r_04!(SimulationSpace:AnimPhysSimSpaceType Enum, ComponentPose=@n68, Alpha=1, GravityOverride:Vector="0,0,-1000", ExternalForce="0,0,0") -> (Pose) {
    angularDampingOverride=0.999
    angularSpringConstant=50
    bAngularSpring=true
    bOverrideAngularDamping=true
    bUseGravityOverride=true
    gravityScale=0.35
    simulationSpace=RootRelative
    alphaBoolBlend {blendOption=Linear}
    boundBone {boneName=dyn_cape_b_r_04}
    componentLinearAccScale=<0.4 0.4 0.4>
    componentLinearVelScale=<0.4 0.4 0.4>
    gravityOverride {z=-1000}
    physicsBodyDefinitions [{
      boundBone {boneName=dyn_cape_b_r_04}
      boxExtents=<16 16 16>
      constraintSetup {angularLimitsMax=<10 0 10>, angularLimitsMin=<-10 -15 -10>, angularTarget=<1 0 0>}
      localJointOffset=<2 0 0>
     }]
   }
   n70 = "Transform (Modify) Bone - Bone: dyn_cape_f_l_01"(ComponentPose=@n69, Alpha=1) {
    scaleMode=BMM_Additive
    scaleSpace=BCS_BoneSpace
    alphaBoolBlend {blendOption=Linear}
    boneToModify {boneName=dyn_cape_f_l_01}
    scale {x=0.4}
   }
   n71 = "Transform (Modify) Bone - Bone: dyn_cape_s_l_01"!(ComponentPose=@n70, Alpha=1) -> (Pose) {
    scaleMode=BMM_Additive
    scaleSpace=BCS_BoneSpace
    alphaBoolBlend {blendOption=Linear}
    boneToModify {boneName=dyn_cape_s_l_01}
    scale {x=0.4}
   }
   n72 = "Transform (Modify) Bone - Bone: dyn_cape_b_l_01"!(ComponentPose=@n71, Alpha=1) -> (Pose) {
    scaleMode=BMM_Additive
    scaleSpace=BCS_BoneSpace
    alphaBoolBlend {blendOption=Linear}
    boneToModify {boneName=dyn_cape_b_l_01}
    scale {x=0.4}
   }
   n73 = "Transform (Modify) Bone - Bone: dyn_cape_f_r_01"(ComponentPose=@n72, Alpha=1) {
    scaleMode=BMM_Additive
    scaleSpace=BCS_BoneSpace
    alphaBoolBlend {blendOption=Linear}
    boneToModify {boneName=dyn_cape_f_r_01}
    scale {x=0.4}
   }
   n74 = "Transform (Modify) Bone - Bone: dyn_cape_s_r_01"!(ComponentPose=@n73, Alpha=1) -> (Pose) {
    scaleMode=BMM_Additive
    scaleSpace=BCS_BoneSpace
    alphaBoolBlend {blendOption=Linear}
    boneToModify {boneName=dyn_cape_s_r_01}
    scale {x=0.4}
   }
   n75 = "Transform (Modify) Bone - Bone: dyn_cape_b_r_01"!(ComponentPose=@n74, Alpha=1) -> (Pose) {
    scaleMode=BMM_Additive
    scaleSpace=BCS_BoneSpace
    alphaBoolBlend {blendOption=Linear}
    boneToModify {boneName=dyn_cape_b_r_01}
    scale {x=0.4}
   }
   n76 = "Transform (Modify) Bone - Bone: dyn_cape_f_l_base"!(ComponentPose=@n75, Alpha:float=1, Translation="0,0,0", Rotation="-5,4,0", Scale="1,1,1") -> (Pose) {
    rotationMode=BMM_Additive
    rotationSpace=BCS_BoneSpace
    alphaBoolBlend {blendOption=Linear}
    boneToModify {boneName=dyn_cape_f_l_base}
    rotation {pitch=-5, yaw=4}
   }
   n77 = "Transform (Modify) Bone - Bone: dyn_cape_f_l_04"(ComponentPose=@n76, Alpha=1, Translation="0,0,0", Rotation="0,-10,0", Scale="1,1,1") {
    rotationMode=BMM_Additive
    rotationSpace=BCS_BoneSpace
    alphaBoolBlend {blendOption=Linear}
    boneToModify {boneName=dyn_cape_f_l_04}
    rotation {yaw=-10}
   }
   n78 = "Transform (Modify) Bone - Bone: dyn_cape_s_l_base"!(ComponentPose=@n77, Alpha:float=1, Translation="0,0,0", Rotation="-4,11,0", Scale="1,1,1") -> (Pose) {
    rotationMode=BMM_Additive
    rotationSpace=BCS_BoneSpace
    alphaBoolBlend {blendOption=Linear}
    boneToModify {boneName=dyn_cape_s_l_base}
    rotation {pitch=-4, yaw=11}
   }
   n79 = "Transform (Modify) Bone - Bone: dyn_cape_b_l_base"!(ComponentPose=@n78, Alpha:float=1, Translation="0,0,0", Rotation="-7,0,0", Scale="1,1,1") -> (Pose) {
    rotationMode=BMM_Additive
    rotationSpace=BCS_BoneSpace
    alphaBoolBlend {blendOption=Linear}
    boneToModify {boneName=dyn_cape_b_l_base}
    rotation {pitch=-7}
   }
   n80 = "Transform (Modify) Bone - Bone: dyn_L_thumb_scale"!(ComponentPose=@n79, Alpha=1, Rotation="0,0,0") -> (Pose) {
    scaleMode=BMM_Additive
    scaleSpace=BCS_BoneSpace
    translationMode=BMM_Additive
    translationSpace=BCS_BoneSpace
    alphaBoolBlend {blendOption=Linear}
    boneToModify {boneName=dyn_L_thumb_scale}
    scale=<0.5 0.8 0.5>
    translation {x=-1.5}
   }
   n81 = "Transform (Modify) Bone - Bone: dyn_L_index_scale"!(ComponentPose=@n80, Alpha=1, Rotation="0,0,0") -> (Pose) {
    scaleMode=BMM_Additive
    scaleSpace=BCS_BoneSpace
    translationMode=BMM_Additive
    translationSpace=BCS_BoneSpace
    alphaBoolBlend {blendOption=Linear}
    boneToModify {boneName=dyn_L_index_scale}
    scale=<0.5 0.8 0.5>
    translation {x=-1.5}
   }
   n82 = "Transform (Modify) Bone - Bone: dyn_L_middle_scale"!(ComponentPose=@n81, Alpha=1, Rotation="0,0,0") -> (Pose) {
    scaleMode=BMM_Additive
    scaleSpace=BCS_BoneSpace
    translationMode=BMM_Additive
    translationSpace=BCS_BoneSpace
    alphaBoolBlend {blendOption=Linear}
    boneToModify {boneName=dyn_L_middle_scale}
    scale=<0.5 0.8 0.5>
    translation {x=-1.5}
   }
   n83 = "Transform (Modify) Bone - Bone: dyn_L_ring_scale"!(ComponentPose=@n82, Alpha=1, Rotation="0,0,0") -> (Pose) {
    scaleMode=BMM_Additive
    scaleSpace=BCS_BoneSpace
    translationMode=BMM_Additive
    translationSpace=BCS_BoneSpace
    alphaBoolBlend {blendOption=Linear}
    boneToModify {boneName=dyn_L_ring_scale}
    scale=<0.5 0.8 0.5>
    translation {x=-1.5}
   }
   n84 = "Transform (Modify) Bone - Bone: dyn_L_pinky_scale"!(ComponentPose=@n83, Alpha=1, Rotation="0,0,0") -> (Pose) {
    scaleMode=BMM_Additive
    scaleSpace=BCS_BoneSpace
    translationMode=BMM_Additive
    translationSpace=BCS_BoneSpace
    alphaBoolBlend {blendOption=Linear}
    boneToModify {boneName=dyn_L_pinky_scale}
    scale=<0.5 0.8 0.5>
    translation {x=-1.5}
   }
   n85 = "Transform (Modify) Bone - Bone: dyn_R_thumb_scale"!(ComponentPose=@n84, Alpha=1, Rotation="0,0,0") -> (Pose) {
    scaleMode=BMM_Additive
    scaleSpace=BCS_BoneSpace
    translationMode=BMM_Additive
    translationSpace=BCS_BoneSpace
    alphaBoolBlend {blendOption=Linear}
    boneToModify {boneName=dyn_R_thumb_scale}
    scale=<0.5 0.8 0.5>
    translation {x=1.5}
   }
   n86 = "Transform (Modify) Bone - Bone: dyn_R_index_scale"!(ComponentPose=@n85, Alpha=1, Rotation="0,0,0") -> (Pose) {
    scaleMode=BMM_Additive
    scaleSpace=BCS_BoneSpace
    translationMode=BMM_Additive
    translationSpace=BCS_BoneSpace
    alphaBoolBlend {blendOption=Linear}
    boneToModify {boneName=dyn_R_index_scale}
    scale=<0.5 0.8 0.5>
    translation {x=1.5}
   }
   n87 = "Transform (Modify) Bone - Bone: dyn_R_middle_scale"!(ComponentPose=@n86, Alpha=1, Rotation="0,0,0") -> (Pose) {
    scaleMode=BMM_Additive
    scaleSpace=BCS_BoneSpace
    translationMode=BMM_Additive
    translationSpace=BCS_BoneSpace
    alphaBoolBlend {blendOption=Linear}
    boneToModify {boneName=dyn_R_middle_scale}
    scale=<0.5 0.8 0.5>
    translation {x=1.5}
   }
   n88 = "Transform (Modify) Bone - Bone: dyn_R_ring_scale"!(ComponentPose=@n87, Alpha=1, Rotation="0,0,0") -> (Pose) {
    scaleMode=BMM_Additive
    scaleSpace=BCS_BoneSpace
    translationMode=BMM_Additive
    translationSpace=BCS_BoneSpace
    alphaBoolBlend {blendOption=Linear}
    boneToModify {boneName=dyn_R_ring_scale}
    scale=<0.5 0.8 0.5>
    translation {x=1.5}
   }
   n89 = "Transform (Modify) Bone - Bone: dyn_R_pinky_scale"!(ComponentPose=@n88, Alpha=1, Rotation="0,0,0") -> (Pose) {
    scaleMode=BMM_Additive
    scaleSpace=BCS_BoneSpace
    translationMode=BMM_Additive
    translationSpace=BCS_BoneSpace
    alphaBoolBlend {blendOption=Linear}
    boneToModify {boneName=dyn_R_pinky_scale}
    scale=<0.5 0.8 0.5>
    translation {x=1.5}
   }
   Output Pose(Result=Component To Local(ComponentPose=@n89)) {layerGroup=None}
   Get bIsKneeSliding()
  ]}
 EventGraph {nodes [
   Event BlueprintUpdateAnimation()
   > n1 = Sequence()
   > n2 = Gravity Override State Function DEPRECATED(CharacterPartAnimInstance=Self-Reference(), GameTimeInSeconds=GetGameTimeInSeconds(), GravityOverrideParams=Get Gravity Override Params())
   > Set Gravity Override(Gravity Override=@n2)
   > Set frontendAlpha(frontendAlpha:double=SelectFloat(A=0.5, B=1, bPickA=Get bIsInFrontEnd())) -> (Output_Get:double)
   > Set kneeSlidingFix(kneeSlidingFix:bool=AND Boolean(A=Get bIsKneeSliding(), B=Get bRigidBodyEnabled())) -> (Output_Get:bool)
   > Set simSpace(simSpace:AnimPhysSimSpaceType Enum) -> (Output_Get:AnimPhysSimSpaceType Enum)
   > Branch(Condition=OR Boolean(A=Get bIsDBNO(), B=Get bIsInFrontEnd()))
   > n3 = Branch(Condition=OR Boolean(A=Get bIsDBNOCarried(), B=Get bIsInFrontEnd()))
   > Set simSpace(simSpace=World)
   Set simSpace(execute=@n3.else)
   n4 = Vehicle Dynamics Scaler Function!(execute=@n1.then_1, CharacterPartAnimInstance:Custom Character Part Anim Instance Object Reference=Self-Reference(), Debug:bool=false) -> (then, InVehicleHeadScaleAlpha:double, InVehicleHeadSimAlpha:double, InVehicleSkirtScaleAlpha:double, InVehicleSkirtSimAlpha:double)
   > Set In Vehicle Skirt Scale Alpha(In Vehicle Skirt Scale Alpha:double=@n4.InVehicleSkirtScaleAlpha) -> (Output_Get:double)
   > Branch(Condition=Get bIsInFrontEnd())
   > Set isPlayingFrontendIdle(isPlayingFrontendIdle:float=GetCurveValue(self=GetOwnerFrontEndAnimInstance(), CurveName=Frontend_Idle_Bypass)) -> (Output_Get:float)
   n5 = Event BlueprintInitializeAnimation()
   n6 = Branch(execute=[@n5.then, @n9.then], Condition=Get bIsInFrontEnd())
   > Set Sim Space Settings(Sim Space Settings:Sim Space Settings Structure=Get Sim Space Settings FrontEnd() -> (Sim Space Settings FrontEnd:Sim Space Settings Structure)) -> (Output_Get:Sim Space Settings Structure)
   Set Sim Space Settings(execute=@n6.else, Sim Space Settings=Get Sim Space Settings BR() -> (Sim Space Settings BR:Sim Space Settings Structure))
   n7 = Event OnBeginJumping()
   Branch(execute=[@n7.then, @n13.then], Condition=NOT Boolean(A=Get bIsSkyDiving()))
   > Set Jump RBAN(Jump RBAN:bool=true) -> (Output_Get:bool)
   n8 = Event OnEndJumping()
   Set Jump RBAN(execute=[@n8.then, @n14.then], Jump RBAN=false)
   n9 = Event OnCharacterAndPickaxeCustomizationLoaded!() -> (
    OutputDelegate
    then
    CustomBaseMesh:Skeletal Mesh Component Object Reference
    CustomParts:Array of Custom Character Part Object References
    CustomMeshParts:Array of Skeletal Mesh Component Budgeted Object References
   ) {is=title K2Node_Event}
   n10 = Event OnParachuteOpened()
   n11 = Make Array("[0]"=(TagName="Cosmetics.MetaTags.DarkStance"))
   Set customDarkStanceGliderDeployed(
    execute=[@n10.then, @n12.then]
    customDarkStanceGliderDeployed:double="To Float (Boolean)"(InBool=GliderItemDefinitionHasAnyMetaTag!(Tags:Array of Gameplay Tag Structures=@n11, bRequiresOpened:bool=true) -> (ReturnValue:bool))
   ) -> (Output_Get:double)
   n12 = Event OnParachuteClosed()
   n13 = Event OnBeginFalling()
   n14 = Event OnEndFalling()
  ]}
]
variables [
 customDarkStanceGliderDeployed {type=number}
 frontendAlpha {type=number}
 Gravity Override {type=Vector}
 Gravity Override Params {
  type=GravityOverrideParamsStruct
  default {
   skydive_FloatingDown {gravityOverride=<0 0 -980>, windAmplitude=<0 0 400>, windFrequency=<0 0 8>, windOffset=<0 0 400>}
   skydive_Parachuting {gravityOverride=<0 -500 -400>, windAmplitude=<100 -100 100>, windFrequency=<8 8 8>}
   skydive_Surfing {gravityOverride=<0 -100 100>, windAmplitude=<300 -300 300>, windFrequency=<10 10 10>}
   tactical_Sprint {gravityOverride=<0 -450 -980>}
  }
 }
 In Vehicle Skirt Scale Alpha {type=number}
 isPlayingFrontendIdle {type=number}
 Jump RBAN {type=boolean}
 kneeSlidingFix {type=boolean}
 Sim Space Settings {
  type=SimSpaceSettings
  default {
   dampingAlpha=1
   maxAngularAcceleration=200
   maxAngularVelocity=1
   maxLinearAcceleration=100
   maxLinearVelocity=50
   velocityScaleZ=0.2
   worldAlpha=1
   externalLinearDragV=<0 0 0>
  }
 }
 Sim Space Settings BR {
  type=SimSpaceSettings
  default {
   dampingAlpha=1
   maxAngularAcceleration=10000
   maxAngularVelocity=10000
   maxLinearAcceleration=10000
   maxLinearVelocity=10000
   velocityScaleZ=1
   worldAlpha=0
   externalLinearDragV=<0 0 0>
  }
 }
 Sim Space Settings FrontEnd {
  type=SimSpaceSettings
  default {
   dampingAlpha=1
   maxAngularAcceleration=10000
   maxAngularVelocity=10000
   maxLinearAcceleration=10000
   maxLinearVelocity=10000
   velocityScaleZ=1
   worldAlpha=0
   externalLinearDragV=<0 0 0>
  }
 }
 simSpace {type=AnimPhysSimSpaceType}
 skelMesh {type=SkeletalMeshComponent}
]
