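# Physics asset of cape colliders for the DarkStance Inferno medium male player body, with no constraints between bodies.
# Capsules cover pelvis, spine, neck and clavicles, and a sphere covers the head; leg, arm and foot bodies have no shapes.
# Four dyn_cape bone chains (b and s, left and right) carry four offset capsules each.
# Many cape capsules reuse the name dyn_cape_b_r_03_capsule, regardless of their bone.
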PhysicsAsset /BRCosmetics/Characters/Player/Male/Medium/Bodies/M_MED_DarkStance_Inferno/Meshes/M_MED_DarkStance_Inferno_CapeColliders
bodies {
 calf_l []
 calf_r []
 clavicle_l [{shapeType=Capsule, length=7.7636, radius=11.1427, center=<9.9642 -0.7741 -1.5923>, rotation=<rot 75.9352 -0.0001 175.5622>}]
 clavicle_r [{shapeType=Capsule, length=7.7636, radius=11.1427, center=<-9.9642 0.7741 1.5923>, rotation=<rot 75.9351 -0.0001 -4.4378>}]
 dyn_cape_b_l_01 [{shapeType=Capsule, length=18, radius=9, shapeName=dyn_cape_b_l_01_capsule, center=<7.2399 0 -8.717>, rotation=<rot 0 0 90>}]
 dyn_cape_b_l_02 [{shapeType=Capsule, length=15.25, radius=12, shapeName=dyn_cape_b_l_02_capsule, center=<8.8007 0 -9.6293>, rotation=<rot 0 0 90>}]
 dyn_cape_b_l_03 [{shapeType=Capsule, length=20, radius=12.5, shapeName=dyn_cape_b_l_03_capsule, center=<5.3282 0 -9.2294>, rotation=<rot 0 0 90>}]
 dyn_cape_b_l_04 [{shapeType=Capsule, length=13.25, radius=12.5, shapeName=dyn_cape_b_r_03_capsule, center=<5.3282 0 -8.3642>, rotation=<rot 0 0 90>}]
 dyn_cape_b_r_01 [{shapeType=Capsule, length=18, radius=10, shapeName=dyn_cape_b_r_01_capsule, center=<7.2399 0 -8.717>, rotation=<rot 0 0 90>}]
 dyn_cape_b_r_02 [{shapeType=Capsule, length=12.5, radius=12, shapeName=dyn_cape_b_r_02_capsule, center=<8.8007 0 -9.6293>, rotation=<rot 0 0 90>}]
 dyn_cape_b_r_03 [{shapeType=Capsule, length=16.75, radius=12.5, shapeName=dyn_cape_b_r_03_capsule, center=<5.3282 0 -9.2294>, rotation=<rot 0 0 90>}]
 dyn_cape_b_r_04 [{shapeType=Capsule, length=22.75, radius=12.5, shapeName=dyn_cape_b_r_03_capsule, center=<5.3282 0 -8.3642>, rotation=<rot 0 0 90>}]
 dyn_cape_s_l_01 [{shapeType=Capsule, length=8.25, radius=12.5, shapeName=dyn_cape_b_r_03_capsule, center=<5.3282 0 -10.5472>, rotation=<rot 0 0 90>}]
 dyn_cape_s_l_02 [{shapeType=Capsule, length=10.25, radius=12.5, shapeName=dyn_cape_b_r_03_capsule, center=<5.3282 0 -9.6465>, rotation=<rot 0 0 90>}]
 dyn_cape_s_l_03 [{shapeType=Capsule, length=16.75, radius=12.5, shapeName=dyn_cape_b_r_03_capsule, center=<5.3282 0 -9.2294>, rotation=<rot 0 0 90>}]
 dyn_cape_s_l_04 [{shapeType=Capsule, length=22.75, radius=12.5, shapeName=dyn_cape_b_r_03_capsule, center=<5.3282 0 -8.3642>, rotation=<rot 0 0 90>}]
 dyn_cape_s_r_01 [{shapeType=Capsule, length=8.25, radius=12.5, shapeName=dyn_cape_b_r_03_capsule, center=<5.3282 0 -12.941>, rotation=<rot 0 0 90>}]
 dyn_cape_s_r_02 [{shapeType=Capsule, length=4.25, radius=12.5, shapeName=dyn_cape_b_r_03_capsule, center=<5.3282 0 -9.6465>, rotation=<rot 0 0 90>}]
 dyn_cape_s_r_03 [{shapeType=Capsule, length=16.25, radius=12.5, shapeName=dyn_cape_b_r_03_capsule, center=<5.3282 0 -9.2294>, rotation=<rot 0 0 90>}]
 dyn_cape_s_r_04 [{shapeType=Capsule, length=13.25, radius=12.5, shapeName=dyn_cape_b_r_03_capsule, center=<5.3282 0 -8.3642>, rotation=<rot 0 0 90>}]
 foot_l []
 foot_r []
 head [{shapeType=Sphere, radius=17.3125, center=<11.028 -0.174 0>}]
 lowerarm_l []
 lowerarm_r []
 neck_02 [{shapeType=Capsule, length=13.4375, radius=7.3175, center=<-0.2642 -0.3408 0>, rotation=<rot 90 -171.1193 -175.3978>}]
 pelvis [{shapeType=Capsule, length=8.6562, radius=15.0286, center=<-2.6215 -5.2615 0>}]
 spine_01 [{shapeType=Capsule, length=13.8634, radius=16.3714, center=<-2.6215 -2.3007 0>}]
 spine_02 [
  {shapeType=Capsule, length=10.5937, radius=10.945, center=<-2.6215 0.4993 0.0001>}
  {shapeType=Capsule, length=29.875, radius=11.1013, center=<2.5024 0.4992 3.9825>, rotation=<rot 90 0.0049 0.0049>}
  {shapeType=Capsule, length=29.875, radius=11.1013, center=<2.4704 0.4991 -3.9825>, rotation=<rot 90 -0.0396 -0.0396>}
 ]
 spine_03 [{shapeType=Capsule, length=13.3146, radius=13.3958, center=<-7.2863 -2.5966 0.0001>}]
 spine_04 [{shapeType=Capsule, length=14.6562, radius=13.6325, center=<2.1467 -1.2748 0.0002>}]
 spine_05 [
  {shapeType=Capsule, length=17.8437, radius=17.195, center=<-1.7514 2.4879 0>}
  {shapeType=Capsule, length=12.7812, radius=17.0625, center=<0.7875 3.1143 0>}
 ]
 thigh_l []
 thigh_r []
 upperarm_l []
 upperarm_r []
}
constraints []
